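# Sound cue for a looping exterior air conditioner spot ambience at the CrewCut military base.
# A single looping wave feeds an endlessly looping enveloper whose cubic pitch curve slowly drifts slightly up and down, then a modulator applies a fixed volume boost.
# It plays as outdoor ambient sound with medium prop emitter attenuation, distance-based spot concurrency and submix sends disabled.
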SoundCue /Hermes/Audio/CrewCut/Spots/SC_Spot_CrewCut_MilitaryBase_Air_Conditioner_Loop_Ext
root=Modulator_1
nodes [
 {
  id=Enveloper_0
  children [WavePlayer_0]
  properties {
   bLoop=true
   bLoopIndefinitely=true
   loopEnd=30
   pitchCurve {editorCurveData {keys [
      {arriveTangent=-0.0033, interpMode=RCIM_Cubic, time=0, value=0.99}
      <at 15 1.01 RCIM_Cubic>
      {interpMode=RCIM_Cubic, leaveTangent=0.0033, time=30, value=0.99}
     ]}}
   volumeCurve {editorCurveData {keys [<at 0 1>]}}
  }
 }
 {id=Modulator_1, children [Enveloper_0], properties {volumeMax=1.3, volumeMin=1.3}}
 {
  id=WavePlayer_0
  properties {bLooping=true, soundWaveAssetPtr=&/Spot_CrewCut_MilitaryBase_Air_Conditioner_Loop_01_Ext}
 }
]
properties {
 attenuationSettings=/Game/Sounds/Attenuation/Fort_Amb_Emitter_Prop_Med_Att
 bEnableSubmixSends=false
 maxDistance=900
 soundClassObject=&Fort_Ambient_Outdoors_Athena
 volumeMultiplier=0.7
 concurrencyOverrides {maxCount=2}
 concurrencySet [/AmbientAudioBase/Concurrency/AmbientConc_4Max_Spot_Distance_Concurency]
 preEffectBusSends [{}]
}
