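# Ambient spot SoundCue for a doorway light on the CrewCut military base double barricade, playing a quiet, grungy fluorescent tube hum.
# A single looping WavePlayer feeds the cue, routed to the Athena ambient props sound class.
# Overridden capsule attenuation keeps the hum local and enables listener focus, while its distance low-pass range starts beyond the cue's max distance.
# A shared ambient spot concurrency set, stored under SprocketPoppy's FaceAnim folder, limits simultaneous instances.
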SoundCue /Hermes/Audio/CrewCut/Spots/SC_Spot_CrewCut_MilitaryBase_BarricadeDouble_Doorway_Light_A
root=WavePlayer_0
nodes [{
  id=WavePlayer_0
  properties {bLooping=true, soundWaveAssetPtr=/AmbientAudioContent/Waves/SpottedLoops/Spot_Fluorescent_Tube_Light_Hum_Grungy_01_Loop}
 }]
properties {
 bOverrideAttenuation=true
 maxDistance=431.2505
 soundClassObject=&Fort_Ambient_Props_Athena
 volumeMultiplier=0.2
 attenuationOverrides {
  attenuationShape=Capsule
  bAttenuateWithLPF=true
  bEnableListenerFocus=true
  falloffDistance=310.9394
  lPFFrequencyAtMax=3400
  lPFRadiusMax=1000
  lPFRadiusMin=500
  nonFocusVolumeAttenuation=0.2
  attenuationShapeExtents=<120.3111 48.9224>
 }
 concurrencyOverrides {maxCount=2, retriggerTime=2, voiceStealReleaseTime=3}
 concurrencySet [/Hermes/Props/SprocketPoppy/Materials/FaceAnim/AmbientConc_3Max_Spot_Concurrency]
}
