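# Looping ambient spot Sound Cue for buffet food at the CrewCut military base, playing on the Athena ambient props sound class.
# It reuses a lava bubble loop, pitched well up with slight random volume variation through a Modulator.
# A Quality Level root feeds that chain to its first two slots and leaves the other two empty.
# A dedicated attenuation asset and a shared ambient spot concurrency set govern playback.
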SoundCue /Hermes/Audio/CrewCut/Spots/SC_Spot_CrewCut_MilitaryBase_BuffetFood_A
root=QualityLevel_0
nodes [
 {id=Modulator_0, children [WavePlayer_1], properties {pitchMax=1.7, pitchMin=1.4, volumeMax=1, volumeMin=0.9}}
 {id=QualityLevel_0, children [Modulator_0,Modulator_0,null,null]}
 {id=WavePlayer_1, properties {bLooping=true, soundWaveAssetPtr=/Game/Sounds/Fort_Starlight/S2/SoundWaves/SpotFX_Lava_Bubbles_Loop}}
]
properties {
 attenuationSettings=&/ATT_CrewCut_MilitaryBase_BuffetFood
 bEnableSubmixSends=false
 maxDistance=600
 soundClassObject=&Fort_Ambient_Props_Athena
 volumeMultiplier=0.6
 attenuationOverrides {
  bAttenuateWithLPF=true
  bEnableListenerFocus=true
  falloffDistance=950
  lPFFrequencyAtMax=3400
  lPFRadiusMax=1000
  lPFRadiusMin=500
  nonFocusVolumeAttenuation=0.2
  attenuationShapeExtents {x=350}
 }
 concurrencyOverrides {maxCount=2, retriggerTime=2, voiceStealReleaseTime=3}
 concurrencySet [/Hermes/Props/SprocketPoppy/Materials/FaceAnim/AmbientConc_3Max_Spot_Concurrency]
}
