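# Looping ambient spot sound cue for a CrewCut military base comm tower, played as outdoor ambience.
# A mixer layers one randomly picked dish-signal loop, slightly lower in volume, with a hologram loop.
# Each layer gets random downward pitch modulation, plus looping envelopes of different lengths that slowly drift pitch and dip volume, varying the static loops.
# A tower-specific cone attenuation and distance-based concurrency limit its reach and voice count.
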SoundCue /Hermes/Audio/CrewCut/Spots/SC_Spot_CrewCut_MilitaryBase_CommTower_A
root=Mixer_0
nodes [
 {
  id=Enveloper_1
  children [WavePlayer_4]
  properties {
   bLoop=true
   bLoopIndefinitely=true
   loopEnd=30
   pitchCurve {editorCurveData {keys [
      {arriveTangent=-0.0033, interpMode=RCIM_Cubic, time=0, value=0.99}
      <at 15 1.01 RCIM_Cubic>
      {interpMode=RCIM_Cubic, leaveTangent=0.0033, time=30, value=0.99}
     ]}}
   volumeCurve {editorCurveData {keys [<at 0 1>]}}
  }
 }
 {
  id=Enveloper_2
  children [Enveloper_1]
  properties {
   bLoop=true
   bLoopIndefinitely=true
   loopEnd=20
   pitchCurve {editorCurveData {keys [{time=0}]}}
   volumeCurve {editorCurveData {keys [
      {arriveTangent=-0.01, interpMode=RCIM_Cubic, time=0, value=1}
      {arriveTangent=-0.1156, interpMode=RCIM_Cubic, leaveTangent=-0.1156, time=2.0525, value=0.9761}
      <at 3.5902 0.5851>
      {arriveTangent=0.0199, interpMode=RCIM_Cubic, leaveTangent=0.0199, time=5.9575, value=0.9387}
      {arriveTangent=0.0044, interpMode=RCIM_Cubic, leaveTangent=0.0044, time=11.9487, value=0.7516}
      {interpMode=RCIM_Cubic, leaveTangent=0.01, time=20, value=1}
     ]}}
  }
 }
 {
  id=Enveloper_4
  children [Modulator_0]
  properties {
   bLoop=true
   bLoopIndefinitely=true
   loopEnd=30
   pitchCurve {editorCurveData {keys [
      {arriveTangent=-0.0033, interpMode=RCIM_Cubic, time=0, value=0.9}
      <at 15 1.01 RCIM_Cubic>
      {interpMode=RCIM_Cubic, leaveTangent=0.0033, time=30, value=0.9}
     ]}}
   volumeCurve {editorCurveData {keys [<at 0 1>]}}
  }
 }
 {
  id=Enveloper_5
  children [Enveloper_4]
  properties {
   bLoop=true
   bLoopIndefinitely=true
   loopEnd=20
   pitchCurve {editorCurveData {keys [{time=0}]}}
   volumeCurve {editorCurveData {keys [
      {arriveTangent=-0.01, interpMode=RCIM_Cubic, time=0, value=1}
      {arriveTangent=-0.1156, interpMode=RCIM_Cubic, leaveTangent=-0.1156, time=2.0525, value=0.9761}
      <at 3.5902 0.5851>
      {arriveTangent=0.0199, interpMode=RCIM_Cubic, leaveTangent=0.0199, time=5.9575, value=0.9387}
      {arriveTangent=0.0044, interpMode=RCIM_Cubic, leaveTangent=0.0044, time=11.9487, value=0.7516}
      {interpMode=RCIM_Cubic, leaveTangent=0.01, time=20, value=1}
     ]}}
  }
 }
 {id=Mixer_0, children [Enveloper_5,Modulator_2], properties {inputVolume [0.8,1]}}
 {id=Modulator_0, children [Random_0], properties {pitchMax=1, pitchMin=0.8, volumeMax=1, volumeMin=1}}
 {id=Modulator_2, children [Enveloper_2], properties {pitchMax=1, pitchMin=0.8, volumeMax=1, volumeMin=1}}
 {id=Random_0, children [WavePlayer_1,WavePlayer_2,WavePlayer_3,WavePlayer_5,WavePlayer_6], properties {weights [1,1,1,1,1]}}
 {id=WavePlayer_1, properties {bLooping=true, soundWaveAssetPtr=/Game/Sounds/Apollo/Biomes/RadarBases/IOAgents_DishSignals_01}}
 {id=WavePlayer_2, properties {bLooping=true, soundWaveAssetPtr=/Game/Sounds/Apollo/Biomes/RadarBases/IOAgents_DishSignals_03}}
 {id=WavePlayer_3, properties {bLooping=true, soundWaveAssetPtr=/Game/Sounds/Apollo/Biomes/RadarBases/IOAgents_DishSignals_04}}
 {id=WavePlayer_4, properties {bLooping=true, soundWaveAssetPtr=/Game/Sounds/Artemis/POI/Reality_Bloom_Neo/Spots/Spot_Neo_Hologram_Loop_14}}
 {id=WavePlayer_5, properties {bLooping=true, soundWaveAssetPtr=/Game/Sounds/Apollo/Biomes/RadarBases/IOAgents_DishSignals_06}}
 {id=WavePlayer_6, properties {bLooping=true, soundWaveAssetPtr=/Game/Sounds/Apollo/Biomes/RadarBases/IOAgents_DishSignals_07}}
]
properties {
 attenuationSettings=&/ATT_CrewCut_MilitaryBase_CommTower_A_Cone
 bEnableSubmixSends=false
 maxDistance=2250
 soundClassObject=&Fort_Ambient_Outdoors_Athena
 volumeMultiplier=0.7
 concurrencyOverrides {maxCount=2}
 concurrencySet [/AmbientAudioBase/Concurrency/AmbientConc_4Max_Spot_Distance_Concurency]
 preEffectBusSends [{}]
}
