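# Looping ambient spot cue for a processing console in the CrewCut military base, with close-range attenuation and limited concurrency.
# A mixer layers a randomly chosen looping bed with a quieter stream of one-shots, endlessly retriggered after random delays from five nested variation groups.
# Looping envelopers add slow pitch and volume drift to both layers.
# Quality-level nodes keep the bed everywhere but drop one-shots from the last two levels.
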SoundCue /Hermes/Audio/CrewCut/Spots/SC_Spot_CrewCut_MilitaryBase_Console_Processing_Loop
root=Mixer_0
nodes [
 {id=Delay_0, children [Random_9], properties {delayMax=8, delayMin=3}}
 {
  id=Enveloper_0
  children [Random_0]
  properties {
   bLoop=true
   bLoopIndefinitely=true
   loopEnd=30
   pitchCurve {editorCurveData {keys [{arriveTangent=-0.0033, interpMode=RCIM_Cubic, time=0}, <at 15 1.05 RCIM_Cubic>, {interpMode=RCIM_Cubic, leaveTangent=0.0033, time=30}]}}
   volumeCurve {editorCurveData {keys [<at 0 1>]}}
  }
 }
 {
  id=Enveloper_1
  children [Looping_0]
  properties {
   bLoop=true
   bLoopIndefinitely=true
   loopEnd=6
   pitchCurve {editorCurveData {keys [{arriveTangent=-0.0033, interpMode=RCIM_Cubic, time=0}, <at 15 1.05 RCIM_Cubic>, {interpMode=RCIM_Cubic, leaveTangent=0.0033, time=30}]}}
   volumeCurve {editorCurveData {keys [<at 0 1>]}}
  }
 }
 {
  id=Enveloper_2
  children [Enveloper_1]
  properties {
   bLoop=true
   bLoopIndefinitely=true
   loopEnd=6
   pitchCurve {editorCurveData {keys [{time=0}]}}
   volumeCurve {editorCurveData {keys [
      {arriveTangent=-0.01, interpMode=RCIM_Cubic, time=0, value=1}
      {arriveTangent=-0.0586, interpMode=RCIM_Cubic, leaveTangent=-0.0586, time=2.0525, value=0.9761}
      <at 3.8792 0.7727>
      {arriveTangent=-0.0026, interpMode=RCIM_Cubic, leaveTangent=-0.0026, time=5.9575, value=0.9387}
      {arriveTangent=0.0044, interpMode=RCIM_Cubic, leaveTangent=0.0044, time=11.9487, value=0.7516}
      {interpMode=RCIM_Cubic, leaveTangent=0.01, time=20, value=1}
     ]}}
  }
 }
 {
  id=Enveloper_3
  children [Enveloper_0]
  properties {
   bLoop=true
   bLoopIndefinitely=true
   loopEnd=20
   pitchCurve {editorCurveData {keys [{time=0}]}}
   volumeCurve {editorCurveData {keys [
      {arriveTangent=-0.01, interpMode=RCIM_Cubic, time=0, value=1}
      {arriveTangent=-0.1156, interpMode=RCIM_Cubic, leaveTangent=-0.1156, time=2.3014, value=0.8437}
      <at 3.5902 0.5851>
      {arriveTangent=0.0199, interpMode=RCIM_Cubic, leaveTangent=0.0199, time=6.4209, value=0.8395}
      {arriveTangent=0.0118, interpMode=RCIM_Cubic, leaveTangent=0.0118, time=11.9487, value=0.7516}
      {interpMode=RCIM_Cubic, leaveTangent=0.01, time=20, value=1}
     ]}}
  }
 }
 {id=Looping_0, children [Delay_0]}
 {id=Mixer_0, children [QualityLevel_1, QualityLevel_0], properties {inputVolume [0.9,0.35]}}
 {id=Modulator_0, children [Enveloper_2], properties {volumeMax=1.3, volumeMin=1.3}}
 {id=Modulator_1, children [Enveloper_3], properties {volumeMax=1.3, volumeMin=1.3}}
 {id=QualityLevel_0, children [Modulator_0,Modulator_0,null,null]}
 {id=QualityLevel_1, children [Modulator_1,Modulator_1,Modulator_1,Modulator_1]}
 {id=Random_0, children [WavePlayer_8, WavePlayer_10, WavePlayer_11, WavePlayer_12], properties {weights [1,1,1,1]}}
 {id=Random_2, children [WavePlayer_16, WavePlayer_17, WavePlayer_18], properties {weights [1,1,1]}}
 {id=Random_3, children [WavePlayer_22, WavePlayer_23, WavePlayer_24], properties {weights [1,1,1]}}
 {id=Random_4, children [WavePlayer_25, WavePlayer_26, WavePlayer_27, WavePlayer_28, WavePlayer_29], properties {weights [1,1,1,1,1]}}
 {id=Random_5, children [WavePlayer_30, WavePlayer_31, WavePlayer_32, WavePlayer_33, WavePlayer_34], properties {weights [1,1,1,1,1]}}
 {id=Random_6, children [WavePlayer_40, WavePlayer_41, WavePlayer_42, WavePlayer_43, WavePlayer_44], properties {weights [1,1,1,1,1]}}
 {id=Random_9, children [Random_2,Random_3,Random_4,Random_5,Random_6], properties {weights [1,1,1,1,1]}}
 {
  id=WavePlayer_10
  properties {bLooping=true, soundWaveAssetPtr=&/Spot_CrewCut_MilitaryBase_Console_Processing_Loop_03}
 }
 {
  id=WavePlayer_11
  properties {bLooping=true, soundWaveAssetPtr=&/Spot_CrewCut_MilitaryBase_Console_Processing_Loop_04}
 }
 {
  id=WavePlayer_12
  properties {bLooping=true, soundWaveAssetPtr=&/Spot_CrewCut_MilitaryBase_Console_Processing_Loop_05}
 }
 {id=WavePlayer_16, properties {soundWaveAssetPtr=&/Spot_CrewCut_MilitaryBase_Console_Processing_Oneshot_1_1}}
 {id=WavePlayer_17, properties {soundWaveAssetPtr=&/Spot_CrewCut_MilitaryBase_Console_Processing_Oneshot_1_2}}
 {id=WavePlayer_18, properties {soundWaveAssetPtr=&/Spot_CrewCut_MilitaryBase_Console_Processing_Oneshot_1_3}}
 {id=WavePlayer_22, properties {soundWaveAssetPtr=&/Spot_CrewCut_MilitaryBase_Console_Processing_Oneshot_2_1}}
 {id=WavePlayer_23, properties {soundWaveAssetPtr=&/Spot_CrewCut_MilitaryBase_Console_Processing_Oneshot_2_2}}
 {id=WavePlayer_24, properties {soundWaveAssetPtr=&/Spot_CrewCut_MilitaryBase_Console_Processing_Oneshot_2_3}}
 {id=WavePlayer_25, properties {soundWaveAssetPtr=&/Spot_CrewCut_MilitaryBase_Console_Processing_Oneshot_3_1}}
 {id=WavePlayer_26, properties {soundWaveAssetPtr=&/Spot_CrewCut_MilitaryBase_Console_Processing_Oneshot_3_2}}
 {id=WavePlayer_27, properties {soundWaveAssetPtr=&/Spot_CrewCut_MilitaryBase_Console_Processing_Oneshot_3_3}}
 {id=WavePlayer_28, properties {soundWaveAssetPtr=&/Spot_CrewCut_MilitaryBase_Console_Processing_Oneshot_3_4}}
 {id=WavePlayer_29, properties {soundWaveAssetPtr=&/Spot_CrewCut_MilitaryBase_Console_Processing_Oneshot_3_5}}
 {id=WavePlayer_30, properties {soundWaveAssetPtr=&/Spot_CrewCut_MilitaryBase_Console_Processing_Oneshot_4_1}}
 {id=WavePlayer_31, properties {soundWaveAssetPtr=&/Spot_CrewCut_MilitaryBase_Console_Processing_Oneshot_4_2}}
 {id=WavePlayer_32, properties {soundWaveAssetPtr=&/Spot_CrewCut_MilitaryBase_Console_Processing_Oneshot_4_3}}
 {id=WavePlayer_33, properties {soundWaveAssetPtr=&/Spot_CrewCut_MilitaryBase_Console_Processing_Oneshot_4_4}}
 {id=WavePlayer_34, properties {soundWaveAssetPtr=&/Spot_CrewCut_MilitaryBase_Console_Processing_Oneshot_4_5}}
 {id=WavePlayer_40, properties {soundWaveAssetPtr=&/Spot_CrewCut_MilitaryBase_Console_Processing_Oneshot_5_1}}
 {id=WavePlayer_41, properties {soundWaveAssetPtr=&/Spot_CrewCut_MilitaryBase_Console_Processing_Oneshot_5_2}}
 {id=WavePlayer_42, properties {soundWaveAssetPtr=&/Spot_CrewCut_MilitaryBase_Console_Processing_Oneshot_5_3}}
 {id=WavePlayer_43, properties {soundWaveAssetPtr=&/Spot_CrewCut_MilitaryBase_Console_Processing_Oneshot_5_4}}
 {id=WavePlayer_44, properties {soundWaveAssetPtr=&/Spot_CrewCut_MilitaryBase_Console_Processing_Oneshot_5_5}}
 {
  id=WavePlayer_8
  properties {bLooping=true, soundWaveAssetPtr=&/Spot_CrewCut_MilitaryBase_Console_Processing_Loop_01}
 }
]
properties {
 attenuationSettings=/Hermes/Props/CrewCut/CrewCut_Table_A/Audio/ATT_Detail_Amb_Loop_Close
 bEnableSubmixSends=false
 maxDistance=900
 soundClassObject=&Fort_Ambient_Props_Athena
 volumeMultiplier=0.5
 attenuationOverrides {
  bAttenuateWithLPF=true
  bEnableListenerFocus=true
  falloffDistance=950
  lPFFrequencyAtMax=3400
  lPFRadiusMax=1000
  lPFRadiusMin=500
  nonFocusVolumeAttenuation=0.2
  attenuationShapeExtents {x=350}
 }
 concurrencyOverrides {maxCount=2, retriggerTime=2, voiceStealReleaseTime=3}
 concurrencySet [/Hermes/Props/SprocketPoppy/Materials/FaceAnim/AmbientConc_3Max_Spot_Concurrency]
}
