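# Looping ambient spot cue for the CrewCut military base data wall, built from holographic table and hologram loops.
# Mixes a base loop, a distance crossfade between random close loops and a distant loop, and a random hologram loop.
# Looping envelopes drift pitch and volume on the first two layers; the crossfade layer fills only two of four quality slots.
# Overridden capsule attenuation adds low-pass, occlusion and listener focus.
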SoundCue /Hermes/Audio/CrewCut/Spots/SC_Spot_CrewCut_MilitaryBase_DataWall_A
root=Mixer_0
nodes [
 {
  id=DistanceCrossFade_0
  children [Random_1,WavePlayer_4]
  properties {crossFadeInput [
    {fadeInDistanceEnd=0, fadeInDistanceStart=0, fadeOutDistanceEnd=1000, fadeOutDistanceStart=500, volume=1}
    {fadeInDistanceEnd=0, fadeInDistanceStart=0, fadeOutDistanceEnd=1500, fadeOutDistanceStart=1000, volume=0.7}
   ]}
 }
 {
  id=Enveloper_0
  children [WavePlayer_0]
  properties {
   bLoop=true
   bLoopIndefinitely=true
   loopEnd=30
   pitchCurve {editorCurveData {keys [{arriveTangent=-0.0033, interpMode=RCIM_Cubic, time=0}, <at 15 1.05 RCIM_Cubic>, {interpMode=RCIM_Cubic, leaveTangent=0.0033, time=30}]}}
   volumeCurve {editorCurveData {keys [<at 0 1>]}}
  }
 }
 {
  id=Enveloper_3
  children [Enveloper_0]
  properties {
   bLoop=true
   bLoopIndefinitely=true
   loopEnd=20
   pitchCurve {editorCurveData {keys [{time=0}]}}
   volumeCurve {editorCurveData {keys [
      {arriveTangent=-0.01, interpMode=RCIM_Cubic, time=0, value=1}
      {arriveTangent=-0.1156, interpMode=RCIM_Cubic, leaveTangent=-0.1156, time=2.3014, value=0.8437}
      <at 3.5902 0.5851>
      {arriveTangent=0.0199, interpMode=RCIM_Cubic, leaveTangent=0.0199, time=6.4209, value=0.8395}
      {arriveTangent=0.0118, interpMode=RCIM_Cubic, leaveTangent=0.0118, time=11.9487, value=0.7516}
      {interpMode=RCIM_Cubic, leaveTangent=0.01, time=20, value=1}
     ]}}
  }
 }
 {
  id=Enveloper_4
  children [DistanceCrossFade_0]
  properties {
   bLoop=true
   bLoopIndefinitely=true
   loopEnd=30
   pitchCurve {editorCurveData {keys [{arriveTangent=-0.0033, interpMode=RCIM_Cubic, time=0}, <at 15 1.05 RCIM_Cubic>, {interpMode=RCIM_Cubic, leaveTangent=0.0033, time=30}]}}
   volumeCurve {editorCurveData {keys [<at 0 1>]}}
  }
 }
 {
  id=Enveloper_5
  children [Enveloper_4]
  properties {
   bLoop=true
   bLoopIndefinitely=true
   loopEnd=20
   pitchCurve {editorCurveData {keys [{time=0}]}}
   volumeCurve {editorCurveData {keys [
      {arriveTangent=-0.01, interpMode=RCIM_Cubic, time=0, value=1}
      {arriveTangent=-0.1156, interpMode=RCIM_Cubic, leaveTangent=-0.1156, time=2.3014, value=0.8437}
      <at 3.5902 0.5851>
      {arriveTangent=0.0199, interpMode=RCIM_Cubic, leaveTangent=0.0199, time=6.4209, value=0.8395}
      {arriveTangent=0.0118, interpMode=RCIM_Cubic, leaveTangent=0.0118, time=11.9487, value=0.7516}
      {interpMode=RCIM_Cubic, leaveTangent=0.01, time=20, value=1}
     ]}}
  }
 }
 {id=Mixer_0, children [QualityLevel_1, QualityLevel_2, Modulator_0], properties {inputVolume [0.65,0.2,1]}}
 {id=Modulator_0, children [Random_0], properties {pitchMin=1}}
 {id=Modulator_1, children [Enveloper_3], properties {volumeMax=1.3, volumeMin=1.3}}
 {id=Modulator_2, children [Enveloper_5], properties {volumeMax=1.3, volumeMin=1.3}}
 {id=QualityLevel_1, children [Modulator_1,Modulator_1,Modulator_1,Modulator_1]}
 {id=QualityLevel_2, children [Modulator_2,Modulator_2,null,null]}
 {id=Random_0, children [WavePlayer_5,WavePlayer_6,WavePlayer_7], properties {weights [1,1,1]}}
 {id=Random_1, children [WavePlayer_1,WavePlayer_2,WavePlayer_3], properties {weights [1,1,1]}}
 {
  id=WavePlayer_0
  properties {bLooping=true, soundWaveAssetPtr=&/Spot_CrewCut_MilitaryBase_Holographic_Table_Base_Loop_01}
 }
 {
  id=WavePlayer_1
  properties {bLooping=true, soundWaveAssetPtr=&/Spot_CrewCut_MilitaryBase_Holographic_Table_Close_Loop_01}
 }
 {
  id=WavePlayer_2
  properties {bLooping=true, soundWaveAssetPtr=&/Spot_CrewCut_MilitaryBase_Holographic_Table_Close_Loop_02}
 }
 {
  id=WavePlayer_3
  properties {bLooping=true, soundWaveAssetPtr=&/Spot_CrewCut_MilitaryBase_Holographic_Table_Close_Loop_03}
 }
 {
  id=WavePlayer_4
  properties {bLooping=true, soundWaveAssetPtr=&/Spot_CrewCut_MilitaryBase_Holographic_Table_Distant_Loop_01}
 }
 {id=WavePlayer_5, properties {bLooping=true, soundWaveAssetPtr=/AmbientAudioContent/Waves/SpottedLoops/Spot_Keys_And_Locks_HologramLoop_01}}
 {id=WavePlayer_6, properties {bLooping=true, soundWaveAssetPtr=/AmbientAudioContent/Waves/SpottedLoops/Spot_Keys_And_Locks_HologramLoop_02}}
 {id=WavePlayer_7, properties {bLooping=true, soundWaveAssetPtr=/AmbientAudioContent/Waves/SpottedLoops/Spot_Keys_And_Locks_HologramLoop_03}}
]
properties {
 attenuationSettings=/AmbientAudioBase/Attenuations/Fort_Amb_Emitter_Prop_Medium_Att
 bEnableSubmixSends=false
 bOverrideAttenuation=true
 maxDistance=1500
 soundClassObject=&Fort_Ambient_Props_Athena
 volumeMultiplier=0.7
 attenuationOverrides {
  attenuationShape=Capsule
  bAttenuateWithLPF=true
  bEnableListenerFocus=true
  bEnableOcclusion=true
  falloffDistance=350
  focusAzimuth=15
  focusVolumeAttenuation=1.25
  lPFFrequencyAtMax=3400
  lPFRadiusMax=1000
  lPFRadiusMin=500
  nonFocusAzimuth=45
  nonFocusVolumeAttenuation=0.5
  occlusionInterpolationTime=1
  occlusionLowPassFilterFrequency=10000
  occlusionTraceChannel=ECC_GameTraceChannel11
  occlusionVolumeAttenuation=0.5
  attenuationShapeExtents=<200 100>
 }
 concurrencyOverrides {maxCount=2, retriggerTime=2, voiceStealReleaseTime=3}
 concurrencySet [/Hermes/Props/SprocketPoppy/Materials/FaceAnim/AmbientConc_3Max_Spot_Concurrency]
}
