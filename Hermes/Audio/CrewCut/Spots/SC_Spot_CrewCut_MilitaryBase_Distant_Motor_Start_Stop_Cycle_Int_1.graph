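# Sound cue for a distant motor start/stop cycle ambient spot at the CrewCut military base, routed to the indoor ambient one-shot class.
# A random node picks one of six wave variations and feeds two nested, endlessly looping envelopers, the inner one sweeping pitch and the outer one gently dipping volume on different cycle lengths.
# A final modulator slightly randomizes volume, while wide one-shot attenuation and concurrency limits govern playback.
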SoundCue /Hermes/Audio/CrewCut/Spots/SC_Spot_CrewCut_MilitaryBase_Distant_Motor_Start_Stop_Cycle_Int_1
root=Modulator_3
nodes [
 {
  id=Enveloper_1
  children [Random_0]
  properties {
   bLoop=true
   bLoopIndefinitely=true
   loopEnd=30
   pitchCurve {editorCurveData {keys [{arriveTangent=-0.0013, interpMode=RCIM_Cubic, time=0}, <at 8 0.98 RCIM_Cubic>, {interpMode=RCIM_Cubic, leaveTangent=0.0013, time=16}]}}
   volumeCurve {editorCurveData {keys [<at 0 1>]}}
  }
 }
 {
  id=Enveloper_5
  children [Enveloper_1]
  properties {
   bLoop=true
   bLoopIndefinitely=true
   loopEnd=20
   pitchCurve {editorCurveData {keys [{time=0}]}}
   volumeCurve {editorCurveData {keys [
      {arriveTangent=-0.01, interpMode=RCIM_Cubic, time=0, value=1}
      <at 6 0.9 RCIM_Cubic>
      {interpMode=RCIM_Cubic, leaveTangent=0.01, time=20, value=1}
     ]}}
  }
 }
 {id=Modulator_3, children [Enveloper_5], properties {volumeMax=0.55, volumeMin=0.5}}
 {
  id=Random_0
  children [WavePlayer_6, WavePlayer_7, WavePlayer_8, WavePlayer_9, WavePlayer_10, WavePlayer_11]
  properties {weights [1,1,1,1,1,1]}
 }
 {id=WavePlayer_10, properties {soundWaveAssetPtr=&/Spot_CrewCut_MilitaryBase_Distant_Motor_Start_Stop_Cycle_1_2}}
 {id=WavePlayer_11, properties {soundWaveAssetPtr=&/Spot_CrewCut_MilitaryBase_Distant_Motor_Start_Stop_Cycle_1_1}}
 {id=WavePlayer_6, properties {soundWaveAssetPtr=&/Spot_CrewCut_MilitaryBase_Distant_Motor_Start_Stop_Cycle_1_6}}
 {id=WavePlayer_7, properties {soundWaveAssetPtr=&/Spot_CrewCut_MilitaryBase_Distant_Motor_Start_Stop_Cycle_1_5}}
 {id=WavePlayer_8, properties {soundWaveAssetPtr=&/Spot_CrewCut_MilitaryBase_Distant_Motor_Start_Stop_Cycle_1_4}}
 {id=WavePlayer_9, properties {soundWaveAssetPtr=&/Spot_CrewCut_MilitaryBase_Distant_Motor_Start_Stop_Cycle_1_3}}
]
properties {
 attenuationSettings=/AmbientAudioBase/Attenuations/Fort_TileAmbient_OneShot_Wider_Att
 bEnableSubmixSends=false
 maxDistance=11000
 soundClassObject=/Game/Sounds/FortSoundClassesAndModes/Fort_Ambient_OneShots_Indoors
 volumeMultiplier=0.8
 concurrencyOverrides {maxCount=2}
 concurrencySet [/AmbientAudioBase/Concurrency/AmbientConc_4Max_Spot_Distance_Concurency]
 preEffectBusSends [{}]
}
