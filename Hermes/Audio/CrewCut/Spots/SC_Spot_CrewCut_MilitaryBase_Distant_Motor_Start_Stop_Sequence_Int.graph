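# Sound cue for an indoor ambient one-shot spot of a distant motor start/stop sequence in the CrewCut military base.
# A Random node picks among seven equally weighted waves, feeding the first two Quality Level slots while the other two are empty.
# An Enveloper adds a short volume fade-in, then a Modulator slightly randomizes the reduced volume.
# Wide ambient one-shot attenuation and distance-based concurrency govern playback, with submix sends disabled.
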SoundCue /Hermes/Audio/CrewCut/Spots/SC_Spot_CrewCut_MilitaryBase_Distant_Motor_Start_Stop_Sequence_Int
root=Modulator_3
nodes [
 {
  id=Enveloper_4
  children [QualityLevel_0]
  properties {pitchCurve {editorCurveData {keys [{time=0}]}}, volumeCurve {editorCurveData {keys [<at 0 0>,<at 0.05 1>]}}}
 }
 {id=Modulator_3, children [Enveloper_4], properties {volumeMax=0.55, volumeMin=0.5}}
 {id=QualityLevel_0, children [Random_1,Random_1,null,null]}
 {
  id=Random_1
  children [WavePlayer_7, WavePlayer_8, WavePlayer_9, WavePlayer_10, WavePlayer_11, WavePlayer_12, WavePlayer_13]
  properties {weights [1,1,1,1,1,1,1]}
 }
 {id=WavePlayer_10, properties {soundWaveAssetPtr=&/Spot_CrewCut_MilitaryBase_0Distant_Motor_Start_Stop_Sequence_04}}
 {id=WavePlayer_11, properties {soundWaveAssetPtr=&/Spot_CrewCut_MilitaryBase_0Distant_Motor_Start_Stop_Sequence_05}}
 {id=WavePlayer_12, properties {soundWaveAssetPtr=&/Spot_CrewCut_MilitaryBase_0Distant_Motor_Start_Stop_Sequence_06}}
 {id=WavePlayer_13, properties {soundWaveAssetPtr=&/Spot_CrewCut_MilitaryBase_0Distant_Motor_Start_Stop_Sequence_07}}
 {id=WavePlayer_7, properties {soundWaveAssetPtr=&/Spot_CrewCut_MilitaryBase_0Distant_Motor_Start_Stop_Sequence_01}}
 {id=WavePlayer_8, properties {soundWaveAssetPtr=&/Spot_CrewCut_MilitaryBase_0Distant_Motor_Start_Stop_Sequence_02}}
 {id=WavePlayer_9, properties {soundWaveAssetPtr=&/Spot_CrewCut_MilitaryBase_0Distant_Motor_Start_Stop_Sequence_03}}
]
properties {
 attenuationSettings=/AmbientAudioBase/Attenuations/Fort_TileAmbient_OneShot_Wider_Att
 bEnableSubmixSends=false
 maxDistance=11000
 soundClassObject=/Game/Sounds/FortSoundClassesAndModes/Fort_Ambient_OneShots_Indoors
 volumeMultiplier=0.8
 concurrencyOverrides {maxCount=2}
 concurrencySet [/AmbientAudioBase/Concurrency/AmbientConc_4Max_Spot_Distance_Concurency]
 preEffectBusSends [{}]
}
