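# Ambient spot Sound Cue for a grungy fluorescent tube hum at a doorway light in the CrewCut military base.
# A looping wave runs through an Oscillator for slow volume modulation, then a Modulator for slight random volume variation.
# The root quality-level node sends two levels to that same chain and leaves two empty.
# Playback is quiet, with occlusion attenuation, distance low-pass filtering, listener focus and a spot concurrency limit.
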SoundCue /Hermes/Audio/CrewCut/Spots/SC_Spot_CrewCut_MilitaryBase_Doorway_Light_A
root=QualityLevel_0
nodes [
 {id=Modulator_0, children [Oscillator_0], properties {volumeMax=1, volumeMin=0.9}}
 {
  id=Oscillator_0
  children [WavePlayer_0]
  properties {amplitudeMax=0.3, amplitudeMin=0.2, bModulateVolume=true, centerMax=1, centerMin=1, frequencyMax=0.1, frequencyMin=0.1}
 }
 {id=QualityLevel_0, children [Modulator_0,Modulator_0,null,null]}
 {
  id=WavePlayer_0
  properties {bLooping=true, soundWaveAssetPtr=/AmbientAudioContent/Waves/SpottedLoops/Spot_Fluorescent_Tube_Light_Hum_Grungy_01_Loop}
 }
]
properties {
 attenuationSettings=/Game/Sounds/Attenuation/Fort_Prop_Med_Occlusion_Att
 bEnableSubmixSends=false
 maxDistance=2000
 soundClassObject=&Fort_Ambient_Props_Athena
 volumeMultiplier=0.1
 attenuationOverrides {
  bAttenuateWithLPF=true
  bEnableListenerFocus=true
  falloffDistance=950
  lPFFrequencyAtMax=3400
  lPFRadiusMax=1000
  lPFRadiusMin=500
  nonFocusVolumeAttenuation=0.2
  attenuationShapeExtents {x=350}
 }
 concurrencyOverrides {maxCount=2, retriggerTime=2, voiceStealReleaseTime=3}
 concurrencySet [/Hermes/Props/SprocketPoppy/Materials/FaceAnim/AmbientConc_3Max_Spot_Concurrency]
}
