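# Ambient spot cue for a flagpole at the CrewCut military base, playing as a looping positional prop emitter.
# It mixes a flag loop, started after a random delay, with one of three randomly chosen flapping-and-ring-chime loops.
# Each layer runs through endlessly looping 20-second volume and 60-second pitch envelopes, so the wind keeps shifting.
# Attenuation overrides add distance low-pass and listener focus, and concurrency limits cap simultaneous instances.
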SoundCue /Hermes/Audio/CrewCut/Spots/SC_Spot_CrewCut_MilitaryBase_FlagPole_A
root=Mixer_2
nodes [
 {id=Delay_0, children [Enveloper_3], properties {delayMax=2}}
 {
  id=Enveloper_0
  children [Mixer_0]
  properties {
   bLoop=true
   bLoopIndefinitely=true
   loopEnd=60
   pitchCurve {editorCurveData {keys [
      {interpMode=RCIM_Cubic, time=0}
      <at 15 0.95 RCIM_Cubic>
      {arriveTangent=0.0072, interpMode=RCIM_Cubic, leaveTangent=0.0072, tangentMode=RCTM_User, time=30}
      <at 45 1.1 RCIM_Cubic>
      {time=60}
     ]}}
   volumeCurve {editorCurveData {keys [<at 0 1>]}}
  }
 }
 {
  id=Enveloper_1
  children [Enveloper_2]
  properties {
   bLoop=true
   bLoopIndefinitely=true
   loopEnd=60
   pitchCurve {editorCurveData {keys [
      {interpMode=RCIM_Cubic, time=0}
      <at 15 0.95 RCIM_Cubic>
      {arriveTangent=0.0072, interpMode=RCIM_Cubic, leaveTangent=0.0072, tangentMode=RCTM_User, time=30}
      {arriveTangent=-0.0082, interpMode=RCIM_Cubic, leaveTangent=-0.0082, tangentMode=RCTM_User, time=45}
      {time=60}
     ]}}
   volumeCurve {editorCurveData {keys [<at 0 1>]}}
  }
 }
 {
  id=Enveloper_2
  children [Mixer_1]
  properties {
   bLoop=true
   bLoopIndefinitely=true
   loopEnd=20
   pitchCurve {editorCurveData {keys [{time=0}]}}
   volumeCurve {editorCurveData {keys [
      {arriveTangent=-0.01, interpMode=RCIM_Cubic, time=0, value=1}
      {arriveTangent=0.0067, interpMode=RCIM_Cubic, leaveTangent=0.0067, tangentMode=RCTM_User, time=1.02, value=0.9994}
      {arriveTangent=-0.0298, interpMode=RCIM_Cubic, leaveTangent=-0.0298, time=3.3866, value=0.5014}
      {arriveTangent=0.4346, interpMode=RCIM_Cubic, leaveTangent=0.4346, tangentMode=RCTM_User, time=5.8515, value=0.8556}
      {arriveTangent=0.0004, interpMode=RCIM_Cubic, leaveTangent=0.0004, tangentMode=RCTM_User, time=6.556, value=1}
      {arriveTangent=0.0015, interpMode=RCIM_Cubic, leaveTangent=0.0015, tangentMode=RCTM_User, time=8.9203, value=1}
      {arriveTangent=-0.013, interpMode=RCIM_Cubic, leaveTangent=-0.013, time=10.8155, value=0.6232}
      {arriveTangent=-0.0014, interpMode=RCIM_Cubic, leaveTangent=0.0091, tangentMode=RCTM_User, time=12.9479, value=0.9476}
      {arriveTangent=0.0129, interpMode=RCIM_Cubic, leaveTangent=0.0129, time=14.8254, value=0.5}
      <at 17 1>
      <at 20 1 RCIM_Cubic>
     ]}}
  }
 }
 {
  id=Enveloper_3
  children [Enveloper_0]
  properties {
   bLoop=true
   bLoopIndefinitely=true
   loopEnd=20
   pitchCurve {editorCurveData {keys [{time=0}]}}
   volumeCurve {editorCurveData {keys [
      {arriveTangent=-0.01, interpMode=RCIM_Cubic, time=0, value=1}
      <at 10 0.4 RCIM_Cubic>
      {interpMode=RCIM_Cubic, leaveTangent=0.01, time=20, value=1}
     ]}}
  }
 }
 {id=Mixer_0, children [WavePlayer_0], properties {inputVolume [3]}}
 {id=Mixer_1, children [Random_2], properties {inputVolume [1.5]}}
 {id=Mixer_2, children [Delay_0,Enveloper_1], properties {inputVolume [1,1]}}
 {id=Random_2, children [WavePlayer_13, WavePlayer_14, WavePlayer_15], properties {weights [1,1,1]}}
 {id=WavePlayer_0, properties {bLooping=true, soundWaveAssetPtr=/AmbientAudioContent/Waves/SpottedLoops/Spot_CTF_FlagLoop_01}}
 {
  id=WavePlayer_13
  properties {bLooping=true, soundWaveAssetPtr=&/Spot_CrewCut_MilitaryBase_Flag_Flapping_Rings_Chiming_In_Wind_01}
 }
 {
  id=WavePlayer_14
  properties {bLooping=true, soundWaveAssetPtr=&/Spot_CrewCut_MilitaryBase_Flag_Flapping_Rings_Chiming_In_Wind_02}
 }
 {
  id=WavePlayer_15
  properties {bLooping=true, soundWaveAssetPtr=&/Spot_CrewCut_MilitaryBase_Flag_Flapping_Rings_Chiming_In_Wind_03}
 }
]
properties {
 attenuationSettings=/AmbientAudioBase/Attenuations/Fort_Amb_Emitter_Prop_Medium_Att
 bEnableSubmixSends=false
 maxDistance=1100
 pitchMultiplier=0.8
 soundClassObject=&Fort_Ambient_Props_Athena
 volumeMultiplier=0.5
 attenuationOverrides {
  bAttenuateWithLPF=true
  bEnableListenerFocus=true
  falloffDistance=950
  lPFFrequencyAtMax=3400
  lPFRadiusMax=1000
  lPFRadiusMin=500
  nonFocusVolumeAttenuation=0.2
  attenuationShapeExtents {x=350}
 }
 concurrencyOverrides {maxCount=2, retriggerTime=2, voiceStealReleaseTime=3}
 concurrencySet [/Hermes/Props/SprocketPoppy/Materials/FaceAnim/AmbientConc_3Max_Spot_Concurrency]
}
