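# Looping ambient spot sound cue for a holographic table prop in the CrewCut military base.
# Mixes a base loop with one of three randomly chosen close loops; the close layer is silent on the last two quality levels.
# Nested looping envelopes on each layer add a 30-second pitch swell and a 20-second volume drift, varying the loop.
# Medium prop attenuation is overridden with low-pass and listener focus, and concurrency limits instances.
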SoundCue /Hermes/Audio/CrewCut/Spots/SC_Spot_CrewCut_MilitaryBase_Holotable_Loop
root=Mixer_0
nodes [
 {
  id=Enveloper_0
  children [Mixer_1]
  properties {
   bLoop=true
   bLoopIndefinitely=true
   loopEnd=30
   pitchCurve {editorCurveData {keys [{arriveTangent=-0.0033, interpMode=RCIM_Cubic, time=0}, <at 15 1.05 RCIM_Cubic>, {interpMode=RCIM_Cubic, leaveTangent=0.0033, time=30}]}}
   volumeCurve {editorCurveData {keys [<at 0 1>]}}
  }
 }
 {
  id=Enveloper_3
  children [Enveloper_0]
  properties {
   bLoop=true
   bLoopIndefinitely=true
   loopEnd=20
   pitchCurve {editorCurveData {keys [{time=0}]}}
   volumeCurve {editorCurveData {keys [
      {arriveTangent=-0.01, interpMode=RCIM_Cubic, time=0, value=1}
      {arriveTangent=-0.0665, interpMode=RCIM_Cubic, leaveTangent=-0.0665, time=2.3014, value=0.8437}
      <at 4.0452 0.7311>
      {arriveTangent=0.0026, interpMode=RCIM_Cubic, leaveTangent=0.0026, time=6.4209, value=0.8395}
      {arriveTangent=0.0118, interpMode=RCIM_Cubic, leaveTangent=0.0118, time=11.9487, value=0.7516}
      {interpMode=RCIM_Cubic, leaveTangent=0.01, time=20, value=1}
     ]}}
  }
 }
 {
  id=Enveloper_4
  children [Random_1]
  properties {
   bLoop=true
   bLoopIndefinitely=true
   loopEnd=30
   pitchCurve {editorCurveData {keys [{arriveTangent=-0.0033, interpMode=RCIM_Cubic, time=0}, <at 15 1.05 RCIM_Cubic>, {interpMode=RCIM_Cubic, leaveTangent=0.0033, time=30}]}}
   volumeCurve {editorCurveData {keys [<at 0 1>]}}
  }
 }
 {
  id=Enveloper_5
  children [Enveloper_4]
  properties {
   bLoop=true
   bLoopIndefinitely=true
   loopEnd=20
   pitchCurve {editorCurveData {keys [{time=0}]}}
   volumeCurve {editorCurveData {keys [
      {arriveTangent=-0.01, interpMode=RCIM_Cubic, time=0, value=1}
      {arriveTangent=-0.0572, interpMode=RCIM_Cubic, leaveTangent=-0.0572, time=2.3014, value=0.8437}
      <at 4.0652 0.7676>
      {arriveTangent=-0.002, interpMode=RCIM_Cubic, leaveTangent=-0.002, time=6.4209, value=0.8395}
      {arriveTangent=0.0118, interpMode=RCIM_Cubic, leaveTangent=0.0118, time=11.9487, value=0.7516}
      {interpMode=RCIM_Cubic, leaveTangent=0.01, time=20, value=1}
     ]}}
  }
 }
 {id=Mixer_0, children [QualityLevel_1, QualityLevel_2], properties {inputVolume [1,1]}}
 {id=Mixer_1, children [WavePlayer_0,null], properties {inputVolume [1,0.25]}}
 {id=Modulator_1, children [Enveloper_3], properties {volumeMax=1.3, volumeMin=1.3}}
 {id=Modulator_2, children [Enveloper_5], properties {volumeMax=1.3, volumeMin=1.3}}
 {id=QualityLevel_1, children [Modulator_1,Modulator_1,Modulator_1,Modulator_1]}
 {id=QualityLevel_2, children [Modulator_2,Modulator_2,null,null]}
 {id=Random_1, children [WavePlayer_1,WavePlayer_2,WavePlayer_3], properties {weights [1,1,1]}}
 {
  id=WavePlayer_0
  properties {bLooping=true, soundWaveAssetPtr=&/Spot_CrewCut_MilitaryBase_Holographic_Table_Base_Loop_01}
 }
 {
  id=WavePlayer_1
  properties {bLooping=true, soundWaveAssetPtr=&/Spot_CrewCut_MilitaryBase_Holographic_Table_Close_Loop_01}
 }
 {
  id=WavePlayer_2
  properties {bLooping=true, soundWaveAssetPtr=&/Spot_CrewCut_MilitaryBase_Holographic_Table_Close_Loop_02}
 }
 {
  id=WavePlayer_3
  properties {bLooping=true, soundWaveAssetPtr=&/Spot_CrewCut_MilitaryBase_Holographic_Table_Close_Loop_03}
 }
]
properties {
 attenuationSettings=/AmbientAudioBase/Attenuations/Fort_Amb_Emitter_Prop_Medium_Att
 bEnableSubmixSends=false
 maxDistance=1100
 priority=5
 soundClassObject=&Fort_Ambient_Props_Athena
 volumeMultiplier=0.6
 attenuationOverrides {
  bAttenuateWithLPF=true
  bEnableListenerFocus=true
  falloffDistance=950
  lPFFrequencyAtMax=3400
  lPFRadiusMax=1000
  lPFRadiusMin=500
  nonFocusVolumeAttenuation=0.2
  attenuationShapeExtents {x=350}
 }
 concurrencyOverrides {maxCount=2, retriggerTime=2, voiceStealReleaseTime=3}
 concurrencySet [/Hermes/Props/SprocketPoppy/Materials/FaceAnim/AmbientConc_3Max_Spot_Concurrency]
}
