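# Looping ambient spot cue for lab equipment in the CrewCut military base, built from four computer loop waves on the ambient props sound class.
# A Random node picks one loop, a short enveloper fades it in, and nested looping envelopers modulate pitch and volume over 30- and 20-second cycles.
# A fixed-gain Modulator feeds the first two quality tiers, leaving the others empty.
# Medium prop attenuation adds low-pass and listener focus, alongside concurrency limits.
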SoundCue /Hermes/Audio/CrewCut/Spots/SC_Spot_CrewCut_MilitaryBase_LabEquipment_B
root=QualityLevel_1
nodes [
 {
  id=Enveloper_0
  children [Enveloper_4]
  properties {
   bLoop=true
   bLoopIndefinitely=true
   loopEnd=30
   pitchCurve {editorCurveData {keys [{arriveTangent=-0.0033, interpMode=RCIM_Cubic, time=0}, <at 15 1.05 RCIM_Cubic>, {interpMode=RCIM_Cubic, leaveTangent=0.0033, time=30}]}}
   volumeCurve {editorCurveData {keys [<at 0 1>]}}
  }
 }
 {
  id=Enveloper_3
  children [Enveloper_0]
  properties {
   bLoop=true
   bLoopIndefinitely=true
   loopEnd=20
   pitchCurve {editorCurveData {keys [{time=0}]}}
   volumeCurve {editorCurveData {keys [
      {arriveTangent=-0.01, interpMode=RCIM_Cubic, time=0, value=1}
      {arriveTangent=-0.1156, interpMode=RCIM_Cubic, leaveTangent=-0.1156, time=2.0525, value=0.9761}
      <at 3.5902 0.5851>
      {arriveTangent=0.0199, interpMode=RCIM_Cubic, leaveTangent=0.0199, time=5.9575, value=0.9387}
      {arriveTangent=0.0044, interpMode=RCIM_Cubic, leaveTangent=0.0044, time=11.9487, value=0.7516}
      {interpMode=RCIM_Cubic, leaveTangent=0.01, time=20, value=1}
     ]}}
  }
 }
 {
  id=Enveloper_4
  children [Random_0]
  properties {pitchCurve {editorCurveData {keys [{time=0}]}}, volumeCurve {editorCurveData {keys [<at 0 0>,<at 0.2 1>]}}}
 }
 {id=Modulator_1, children [Enveloper_3], properties {volumeMax=1.3, volumeMin=1.3}}
 {id=QualityLevel_1, children [Modulator_1,Modulator_1,null,null]}
 {id=Random_0, children [WavePlayer_0,WavePlayer_4,WavePlayer_5,WavePlayer_6], properties {weights [1,1,1,1]}}
 {id=WavePlayer_0, properties {bLooping=true, soundWaveAssetPtr=/Hermes/Props/CrewCut/CrewCut_Table_A/Audio/TidalCrane_ComputerA_Loop_01}}
 {id=WavePlayer_4, properties {bLooping=true, soundWaveAssetPtr=/Hermes/Props/CrewCut/CrewCut_Table_A/Audio/TidalCrane_ComputerA_Loop_02}}
 {id=WavePlayer_5, properties {bLooping=true, soundWaveAssetPtr=/Hermes/Props/CrewCut/CrewCut_Table_A/Audio/TidalCrane_ComputerA_Loop_03}}
 {id=WavePlayer_6, properties {bLooping=true, soundWaveAssetPtr=/Hermes/Props/CrewCut/CrewCut_Table_A/Audio/TidalCrane_ComputerA_Loop_04}}
]
properties {
 attenuationSettings=/AmbientAudioBase/Attenuations/Fort_Amb_Emitter_Prop_Medium_Att
 bEnableSubmixSends=false
 maxDistance=1100
 soundClassObject=&Fort_Ambient_Props_Athena
 volumeMultiplier=0.5
 attenuationOverrides {
  bAttenuateWithLPF=true
  bEnableListenerFocus=true
  falloffDistance=950
  lPFFrequencyAtMax=3400
  lPFRadiusMax=1000
  lPFRadiusMin=500
  nonFocusVolumeAttenuation=0.2
  attenuationShapeExtents {x=350}
 }
 concurrencyOverrides {maxCount=2, retriggerTime=2, voiceStealReleaseTime=3}
 concurrencySet [/Hermes/Props/SprocketPoppy/Materials/FaceAnim/AmbientConc_3Max_Spot_Concurrency]
}
