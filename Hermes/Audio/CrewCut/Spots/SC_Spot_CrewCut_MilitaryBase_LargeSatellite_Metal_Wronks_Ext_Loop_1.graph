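# Looping exterior spot ambience cue of metal creaks and groans for the CrewCut military base large satellite, reusing Big Bridge samples.
# Three looping layers of creaks, large creaks and stress groans each play random waves after random delays into a mixer.
# Two nested looping envelopes shape pitch and volume, then a modulator randomizes pitch.
# Two quality levels are silent; overrides add focus-aware low-pass attenuation and concurrency limits, plus a distant reverb send.
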SoundCue /Hermes/Audio/CrewCut/Spots/SC_Spot_CrewCut_MilitaryBase_LargeSatellite_Metal_Wronks_Ext_Loop_1
root=QualityLevel_0
nodes [
 {id=Delay_0, children [Random_1], properties {delayMax=30, delayMin=5}}
 {id=Delay_1, children [Random_2], properties {delayMax=30, delayMin=5}}
 {id=Delay_2, children [Random_3], properties {delayMax=30, delayMin=5}}
 {
  id=Enveloper_0
  children [Mixer_0]
  properties {
   bLoop=true
   bLoopIndefinitely=true
   loopEnd=30
   pitchCurve {editorCurveData {keys [
      {arriveTangent=-0.0033, interpMode=RCIM_Cubic, time=0, value=0.99}
      <at 15 1.01 RCIM_Cubic>
      {interpMode=RCIM_Cubic, leaveTangent=0.0033, time=30, value=0.99}
     ]}}
   volumeCurve {editorCurveData {keys [<at 0 1>]}}
  }
 }
 {
  id=Enveloper_3
  children [Enveloper_0]
  properties {
   bLoop=true
   bLoopIndefinitely=true
   loopEnd=20
   pitchCurve {editorCurveData {keys [{time=0}]}}
   volumeCurve {editorCurveData {keys [
      {arriveTangent=-0.01, interpMode=RCIM_Cubic, time=0, value=1}
      {arriveTangent=-0.1156, interpMode=RCIM_Cubic, leaveTangent=-0.1156, time=2.0525, value=0.9761}
      <at 3.5902 0.5851>
      {arriveTangent=0.0199, interpMode=RCIM_Cubic, leaveTangent=0.0199, time=5.9575, value=0.9387}
      {arriveTangent=0.0044, interpMode=RCIM_Cubic, leaveTangent=0.0044, time=11.9487, value=0.7516}
      {interpMode=RCIM_Cubic, leaveTangent=0.01, time=20, value=1}
     ]}}
  }
 }
 {id=Looping_0, children [Delay_2]}
 {id=Looping_1, children [Delay_1]}
 {id=Looping_2, children [Delay_0]}
 {id=Mixer_0, children [Looping_2,Looping_1,Looping_0], properties {inputVolume [1,1,1]}}
 {id=Modulator_1, children [Enveloper_3], properties {pitchMax=1.1, pitchMin=0.8, volumeMax=1.3, volumeMin=1.3}}
 {id=QualityLevel_0, children [Modulator_1,Modulator_1,null,null]}
 {id=Random_1, children [WavePlayer_1,WavePlayer_2,WavePlayer_3,WavePlayer_4,WavePlayer_5], properties {weights [1,1,1,1,1]}}
 {id=Random_2, children [WavePlayer_6,WavePlayer_7,WavePlayer_8], properties {weights [1,1,1]}}
 {
  id=Random_3
  children [WavePlayer_9, WavePlayer_10, WavePlayer_11, WavePlayer_12, WavePlayer_13, WavePlayer_14, WavePlayer_15]
  properties {weights [1,1,1,1,1,1,1]}
 }
 {id=WavePlayer_1, properties {soundWaveAssetPtr=/Game/Sounds/Artemis/POI/Big_Bridge/Spots/Spot_Big_Bridge_Metal_Creak_01}}
 {id=WavePlayer_10, properties {soundWaveAssetPtr=/Game/Sounds/Artemis/POI/Big_Bridge/Spots/Spot_Big_Metal_Bridge_Huge_Stress_Groan_02}}
 {id=WavePlayer_11, properties {soundWaveAssetPtr=/Game/Sounds/Artemis/POI/Big_Bridge/Spots/Spot_Big_Metal_Bridge_Huge_Stress_Groan_03}}
 {id=WavePlayer_12, properties {soundWaveAssetPtr=/Game/Sounds/Artemis/POI/Big_Bridge/Spots/Spot_Big_Metal_Bridge_Huge_Stress_Groan_04}}
 {id=WavePlayer_13, properties {soundWaveAssetPtr=/Game/Sounds/Artemis/POI/Big_Bridge/Spots/Spot_Big_Metal_Bridge_Huge_Stress_Groan_05}}
 {id=WavePlayer_14, properties {soundWaveAssetPtr=/Game/Sounds/Artemis/POI/Big_Bridge/Spots/Spot_Big_Metal_Bridge_Huge_Stress_Groan_06}}
 {id=WavePlayer_15, properties {soundWaveAssetPtr=/Game/Sounds/Artemis/POI/Big_Bridge/Spots/Spot_Big_Metal_Bridge_Huge_Stress_Groan_07}}
 {id=WavePlayer_2, properties {soundWaveAssetPtr=/Game/Sounds/Artemis/POI/Big_Bridge/Spots/Spot_Big_Bridge_Metal_Creak_02}}
 {id=WavePlayer_3, properties {soundWaveAssetPtr=/Game/Sounds/Artemis/POI/Big_Bridge/Spots/Spot_Big_Bridge_Metal_Creak_03}}
 {id=WavePlayer_4, properties {soundWaveAssetPtr=/Game/Sounds/Artemis/POI/Big_Bridge/Spots/Spot_Big_Bridge_Metal_Creak_04}}
 {id=WavePlayer_5, properties {soundWaveAssetPtr=/Game/Sounds/Artemis/POI/Big_Bridge/Spots/Spot_Big_Bridge_Metal_Creak_05}}
 {id=WavePlayer_6, properties {soundWaveAssetPtr=/Game/Sounds/Artemis/POI/Big_Bridge/Spots/Spot_Big_Bridge_Metal_Creak_Large_01}}
 {id=WavePlayer_7, properties {soundWaveAssetPtr=/Game/Sounds/Artemis/POI/Big_Bridge/Spots/Spot_Big_Bridge_Metal_Creak_Large_02}}
 {id=WavePlayer_8, properties {soundWaveAssetPtr=/Game/Sounds/Artemis/POI/Big_Bridge/Spots/Spot_Big_Bridge_Metal_Creak_Large_03}}
 {id=WavePlayer_9, properties {soundWaveAssetPtr=/Game/Sounds/Artemis/POI/Big_Bridge/Spots/Spot_Big_Metal_Bridge_Huge_Stress_Groan_01}}
]
properties {
 bOverrideAttenuation=true
 bOverrideConcurrency=true
 maxDistance=2000
 soundClassObject=&Fort_Ambient_Outdoors_Athena
 volumeMultiplier=0.7
 attenuationOverrides {
  bAttenuateWithLPF=true
  bEnableFocusInterpolation=true
  bEnableListenerFocus=true
  falloffDistance=1500
  focusAttackInterpSpeed=0.05
  focusReleaseInterpSpeed=0.01
  lPFFrequencyAtMax=8000
  lPFRadiusMax=1500
  lPFRadiusMin=750
  nonFocusVolumeAttenuation=0.2
  attenuationShapeExtents {x=500}
 }
 concurrencyOverrides {maxCount=2}
 concurrencySet [/AmbientAudioBase/Concurrency/AmbientConc_4Max_Spot_Distance_Concurency]
 preEffectBusSends [{}]
 soundSubmixSends [{minSendLevel=0, sendLevel=0.5, soundSubmix=&Fortnite_Reverb_Distant}]
}
