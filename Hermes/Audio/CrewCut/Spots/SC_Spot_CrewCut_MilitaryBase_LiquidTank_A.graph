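# Ambient spot SoundCue for a liquid tank in the CrewCut military base, with its own attenuation and a concurrency limit.
# A mixer layers boosted random "Stop" waves, looped after an initial delay, with vat and inside-flow loops, the latter pitch-varied.
# Looping envelopers fade both loops in and out on a shared cycle, with the delay timed near the fade-out.
# Only the first two quality levels connect to the mix; the other two are empty.
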SoundCue /Hermes/Audio/CrewCut/Spots/SC_Spot_CrewCut_MilitaryBase_LiquidTank_A
root=QualityLevel_0
nodes [
 {id=Delay_1, children [Looping_2], properties {delayMax=12, delayMin=12}}
 {
  id=Enveloper_0
  children [WavePlayer_4]
  properties {
   bLoop=true
   bLoopIndefinitely=true
   loopEnd=24
   pitchCurve {editorCurveData {keys [{time=0}]}}
   volumeCurve {editorCurveData {keys [
      {arriveTangent=0.6828, interpMode=RCIM_Cubic, leaveTangent=0.6828, tangentMode=RCTM_User, time=0, value=0}
      {arriveTangent=-0.0017, interpMode=RCIM_Cubic, leaveTangent=-0.0017, tangentMode=RCTM_User, time=2, value=1}
      {arriveTangent=-0.0017, interpMode=RCIM_Cubic, leaveTangent=-0.0017, tangentMode=RCTM_User, time=12, value=1}
      <at 14 0>
      <at 24 0>
     ]}}
  }
 }
 {
  id=Enveloper_1
  children [WavePlayer_5]
  properties {
   bLoop=true
   bLoopIndefinitely=true
   loopEnd=24
   pitchCurve {editorCurveData {keys [{time=0}]}}
   volumeCurve {editorCurveData {keys [
      {arriveTangent=0.6828, interpMode=RCIM_Cubic, leaveTangent=0.6828, tangentMode=RCTM_User, time=0, value=0}
      {arriveTangent=-0.0017, interpMode=RCIM_Cubic, leaveTangent=-0.0017, tangentMode=RCTM_User, time=2, value=1}
      {arriveTangent=-0.0017, interpMode=RCIM_Cubic, leaveTangent=-0.0017, tangentMode=RCTM_User, time=12, value=1}
      <at 14 0>
      <at 24 0>
     ]}}
  }
 }
 {id=Looping_2, children [Random_1]}
 {id=Mixer_0, children [Modulator_0,Modulator_7,Modulator_1], properties {inputVolume [4,1,1]}}
 {id=Modulator_0, children [Delay_1]}
 {id=Modulator_1, children [Enveloper_1], properties {pitchMax=1.2, pitchMin=1, volumeMax=1}}
 {id=Modulator_7, children [Enveloper_0]}
 {id=QualityLevel_0, children [Mixer_0,Mixer_0,null,null]}
 {id=Random_1, children [WavePlayer_0,WavePlayer_2,WavePlayer_3], properties {weights [1,1,1]}}
 {id=WavePlayer_0, properties {soundWaveAssetPtr=&/Spot_CrewCut_MilitaryBase_LiquidVat_Stop_04}}
 {id=WavePlayer_2, properties {soundWaveAssetPtr=&/Spot_CrewCut_MilitaryBase_LiquidVat_Stop_02}}
 {id=WavePlayer_3, properties {soundWaveAssetPtr=&/Spot_CrewCut_MilitaryBase_LiquidVat_Stop_01}}
 {id=WavePlayer_4, properties {bLooping=true, soundWaveAssetPtr=&/Spot_CrewCut_MilitaryBase_LiquidVat_Loop}}
 {
  id=WavePlayer_5
  properties {bLooping=true, soundWaveAssetPtr=&/Spot_CrewCut_MilitaryBase_LiquidVat_Flowing_Inside_Loop}
 }
]
properties {
 attenuationSettings=&/ATT_CrewCut_MilitaryBase_LiquidVat
 bEnableBusSends=false
 bEnableSubmixSends=false
 bOverrideConcurrency=true
 maxDistance=650
 soundClassObject=&Fort_Ambient_Outdoors_Athena
 volumeMultiplier=0.15
 concurrencyOverrides {maxCount=2, voiceStealReleaseTime=0.2}
}
