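# Indoor ambient one-shot cue for long, thin metal ticking creak spot sounds in the CrewCut military base.
# A Random node picks one of six wave variants, which passes through a short volume fade-in envelope and slight volume modulation.
# All four quality-level inputs share that chain, feeding a mixer that boosts the level.
# Wide one-shot attenuation, a shared distance-based spot concurrency set and a distant reverb send shape playback.
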SoundCue /Hermes/Audio/CrewCut/Spots/SC_Spot_CrewCut_MilitaryBase_Long_Thin_Metal_Ticking_Creak_Int
root=Mixer_0
nodes [
 {
  id=Enveloper_4
  children [Random_5]
  properties {pitchCurve {editorCurveData {keys [{time=0}]}}, volumeCurve {editorCurveData {keys [<at 0 0>,<at 0.2094 1>]}}}
 }
 {id=Mixer_0, children [QualityLevel_1], properties {inputVolume [3]}}
 {id=Modulator_3, children [Enveloper_4], properties {volumeMax=0.55, volumeMin=0.5}}
 {id=QualityLevel_1, children [Modulator_3,Modulator_3,Modulator_3,Modulator_3]}
 {
  id=Random_5
  children [WavePlayer_46, WavePlayer_47, WavePlayer_48, WavePlayer_49, WavePlayer_50, WavePlayer_51]
  properties {weights [1,1,1,1,1,1]}
 }
 {id=WavePlayer_46, properties {soundWaveAssetPtr=&/Spot_CrewCut_MilitaryBase_Long_Thin_Metal_Ticking_Creak_06}}
 {id=WavePlayer_47, properties {soundWaveAssetPtr=&/Spot_CrewCut_MilitaryBase_Long_Thin_Metal_Ticking_Creak_05}}
 {id=WavePlayer_48, properties {soundWaveAssetPtr=&/Spot_CrewCut_MilitaryBase_Long_Thin_Metal_Ticking_Creak_04}}
 {id=WavePlayer_49, properties {soundWaveAssetPtr=&/Spot_CrewCut_MilitaryBase_Long_Thin_Metal_Ticking_Creak_03}}
 {id=WavePlayer_50, properties {soundWaveAssetPtr=&/Spot_CrewCut_MilitaryBase_Long_Thin_Metal_Ticking_Creak_02}}
 {id=WavePlayer_51, properties {soundWaveAssetPtr=&/Spot_CrewCut_MilitaryBase_Long_Thin_Metal_Ticking_Creak_01}}
]
properties {
 attenuationSettings=/AmbientAudioBase/Attenuations/Fort_TileAmbient_OneShot_Wider_Att
 maxDistance=11000
 soundClassObject=/Game/Sounds/FortSoundClassesAndModes/Fort_Ambient_OneShots_Indoors
 volumeMultiplier=1.5
 concurrencyOverrides {maxCount=2}
 concurrencySet [/AmbientAudioBase/Concurrency/AmbientConc_4Max_Spot_Distance_Concurency]
 preEffectBusSends [{}]
 soundSubmixSends [{minSendLevel=0, sendLevel=0.6, soundSubmix=&Fortnite_Reverb_Distant}]
}
