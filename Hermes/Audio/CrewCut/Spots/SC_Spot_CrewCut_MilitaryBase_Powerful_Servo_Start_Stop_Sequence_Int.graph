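# Interior ambient one-shot sound cue for a powerful servo start/stop spot effect in the CrewCut military base.
# An equally weighted random pick of four waves feeds the first two quality-level slots; the other two are empty.
# An enveloper adds a brief volume fade-in, then a modulator slightly varies the level.
# It uses wide one-shot attenuation, the indoor ambient one-shot sound class and distance-based concurrency limits.
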SoundCue /Hermes/Audio/CrewCut/Spots/SC_Spot_CrewCut_MilitaryBase_Powerful_Servo_Start_Stop_Sequence_Int
root=Modulator_3
nodes [
 {
  id=Enveloper_4
  children [QualityLevel_0]
  properties {pitchCurve {editorCurveData {keys [{time=0}]}}, volumeCurve {editorCurveData {keys [<at 0 0>,<at 0.05 1>]}}}
 }
 {id=Modulator_3, children [Enveloper_4], properties {volumeMax=0.55, volumeMin=0.5}}
 {id=QualityLevel_0, children [Random_0,Random_0,null,null]}
 {id=Random_0, children [WavePlayer_4, WavePlayer_5, WavePlayer_6, WavePlayer_14], properties {weights [1,1,1,1]}}
 {
  id=WavePlayer_14
  properties {soundWaveAssetPtr=&/Spot_CrewCut_MilitaryBase_0Powerful_Servo_Start_Stop_Sequence_01}
 }
 {id=WavePlayer_4, properties {soundWaveAssetPtr=&/Spot_CrewCut_MilitaryBase_0Powerful_Servo_Start_Stop_Sequence_04}}
 {id=WavePlayer_5, properties {soundWaveAssetPtr=&/Spot_CrewCut_MilitaryBase_0Powerful_Servo_Start_Stop_Sequence_03}}
 {id=WavePlayer_6, properties {soundWaveAssetPtr=&/Spot_CrewCut_MilitaryBase_0Powerful_Servo_Start_Stop_Sequence_02}}
]
properties {
 attenuationSettings=/AmbientAudioBase/Attenuations/Fort_TileAmbient_OneShot_Wider_Att
 bEnableSubmixSends=false
 maxDistance=11000
 soundClassObject=/Game/Sounds/FortSoundClassesAndModes/Fort_Ambient_OneShots_Indoors
 volumeMultiplier=0.8
 concurrencyOverrides {maxCount=2}
 concurrencySet [/AmbientAudioBase/Concurrency/AmbientConc_4Max_Spot_Distance_Concurency]
 preEffectBusSends [{}]
}
