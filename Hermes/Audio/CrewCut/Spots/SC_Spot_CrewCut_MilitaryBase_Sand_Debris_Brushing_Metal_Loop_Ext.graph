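# Ambient spot SoundCue named for a looping exterior sand-and-debris brushing-on-metal sound at the CrewCut military base, classed as outdoor ambience.
# A randomly delayed loop feeds a short volume fade-in and random volume modulation, though the delay's child is empty, so no sound wave is connected.
# Only the first two quality-level slots use this chain before the mixer.
# Playback is quiet, attenuated by an exterior metal-rattle occlusion asset and concurrency-limited.
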SoundCue /Hermes/Audio/CrewCut/Spots/SC_Spot_CrewCut_MilitaryBase_Sand_Debris_Brushing_Metal_Loop_Ext
root=Mixer_0
nodes [
 {id=Delay_0, children [null], properties {delayMax=15, delayMin=5}}
 {
  id=Enveloper_4
  children [Looping_0]
  properties {pitchCurve {editorCurveData {keys [{time=0}]}}, volumeCurve {editorCurveData {keys [<at 0 0>,<at 0.2094 1>]}}}
 }
 {id=Looping_0, children [Delay_0]}
 {id=Mixer_0, children [QualityLevel_1], properties {inputVolume [1]}}
 {id=Modulator_3, children [Enveloper_4], properties {volumeMax=0.55, volumeMin=0.5}}
 {id=QualityLevel_1, children [Modulator_3,Modulator_3,null,null]}
]
properties {
 attenuationSettings=/Game/Sounds/Artemis/POI/Hollow_Base/Spots/Spot_Exterior_Metal_Rattles_Occlusion_Att
 bEnableSubmixSends=false
 maxDistance=1625.1821
 soundClassObject=&Fort_Ambient_Outdoors_Athena
 volumeMultiplier=0.35
 concurrencyOverrides {maxCount=2}
 concurrencySet [/AmbientAudioBase/Concurrency/AmbientConc_4Max_Spot_Distance_Concurency]
 preEffectBusSends [{}]
}
