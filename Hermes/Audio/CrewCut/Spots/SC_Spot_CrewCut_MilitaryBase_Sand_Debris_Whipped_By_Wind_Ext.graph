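# Sound cue for an outdoor ambient spot one-shot of sand and debris whipped by wind at the CrewCut military base.
# An equal-weight random pick of thirteen wave variations passes through an Enveloper for a short volume fade-in, then a Modulator for slight variation.
# A Quality Level node routes that chain to its first two levels and leaves the other two empty.
# Close one-shot attenuation and ambient spot concurrency limit its reach and voice count.
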SoundCue /Hermes/Audio/CrewCut/Spots/SC_Spot_CrewCut_MilitaryBase_Sand_Debris_Whipped_By_Wind_Ext
root=Mixer_0
nodes [
 {
  id=Enveloper_4
  children [Random_0]
  properties {pitchCurve {editorCurveData {keys [{time=0}]}}, volumeCurve {editorCurveData {keys [<at 0 0>,<at 0.2094 1>]}}}
 }
 {id=Mixer_0, children [QualityLevel_1], properties {inputVolume [1]}}
 {id=Modulator_3, children [Enveloper_4], properties {volumeMax=0.55, volumeMin=0.5}}
 {id=QualityLevel_1, children [Modulator_3,Modulator_3,null,null]}
 {
  id=Random_0
  children [
   WavePlayer_0
   WavePlayer_1
   WavePlayer_2
   WavePlayer_3
   WavePlayer_4
   WavePlayer_5
   WavePlayer_6
   WavePlayer_7
   WavePlayer_8
   WavePlayer_9
   WavePlayer_10
   WavePlayer_11
   WavePlayer_12
  ]
  properties {weights [1,1,1,1,1,1,1,1,1,1,1,1,1]}
 }
 {id=WavePlayer_0, properties {soundWaveAssetPtr=&/Spot_CrewCut_MilitaryBase_Sand_Debris_Whipped_By_Wind_01}}
 {id=WavePlayer_1, properties {soundWaveAssetPtr=&/Spot_CrewCut_MilitaryBase_Sand_Debris_Whipped_By_Wind_02}}
 {id=WavePlayer_10, properties {soundWaveAssetPtr=&/Spot_CrewCut_MilitaryBase_Sand_Debris_Whipped_By_Wind_11}}
 {id=WavePlayer_11, properties {soundWaveAssetPtr=&/Spot_CrewCut_MilitaryBase_Sand_Debris_Whipped_By_Wind_12}}
 {id=WavePlayer_12, properties {soundWaveAssetPtr=&/Spot_CrewCut_MilitaryBase_Sand_Debris_Whipped_By_Wind_13}}
 {id=WavePlayer_2, properties {soundWaveAssetPtr=&/Spot_CrewCut_MilitaryBase_Sand_Debris_Whipped_By_Wind_03}}
 {id=WavePlayer_3, properties {soundWaveAssetPtr=&/Spot_CrewCut_MilitaryBase_Sand_Debris_Whipped_By_Wind_04}}
 {id=WavePlayer_4, properties {soundWaveAssetPtr=&/Spot_CrewCut_MilitaryBase_Sand_Debris_Whipped_By_Wind_05}}
 {id=WavePlayer_5, properties {soundWaveAssetPtr=&/Spot_CrewCut_MilitaryBase_Sand_Debris_Whipped_By_Wind_06}}
 {id=WavePlayer_6, properties {soundWaveAssetPtr=&/Spot_CrewCut_MilitaryBase_Sand_Debris_Whipped_By_Wind_07}}
 {id=WavePlayer_7, properties {soundWaveAssetPtr=&/Spot_CrewCut_MilitaryBase_Sand_Debris_Whipped_By_Wind_08}}
 {id=WavePlayer_8, properties {soundWaveAssetPtr=&/Spot_CrewCut_MilitaryBase_Sand_Debris_Whipped_By_Wind_09}}
 {id=WavePlayer_9, properties {soundWaveAssetPtr=&/Spot_CrewCut_MilitaryBase_Sand_Debris_Whipped_By_Wind_10}}
]
properties {
 attenuationSettings=/Game/Sounds/Attenuation/Fort_TileAmbient_OneShot_Att_Closer
 bEnableSubmixSends=false
 maxDistance=8500
 soundClassObject=&Fort_Ambient_Outdoors_Athena
 volumeMultiplier=0.35
 concurrencyOverrides {maxCount=2}
 concurrencySet [/AmbientAudioBase/Concurrency/AmbientConc_4Max_Spot_Distance_Concurency]
 preEffectBusSends [{}]
}
