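# Sound cue for short metal creak spot ambience in the CrewCut military base interior, played as an attenuated indoor one-shot.
# A quality-level node picks from twelve random waves at the first level and six at the other three.
# An enveloper adds a brief fade-in, then a root modulator randomizes pitch and volume.
# Concurrency settings cap simultaneous creaks.
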SoundCue /Hermes/Audio/CrewCut/Spots/SC_Spot_CrewCut_MilitaryBase_Short_Metal_Creaks_Int
root=Modulator_3
nodes [
 {
  id=Enveloper_4
  children [QualityLevel_0]
  properties {pitchCurve {editorCurveData {keys [{time=0}]}}, volumeCurve {editorCurveData {keys [<at 0 0>,<at 0.05 1>]}}}
 }
 {id=Modulator_3, children [Enveloper_4], properties {pitchMin=0.8, volumeMax=1.1}}
 {id=QualityLevel_0, children [Random_0,Random_4,Random_4,Random_4]}
 {
  id=Random_0
  children [
   WavePlayer_40
   WavePlayer_41
   WavePlayer_42
   WavePlayer_43
   WavePlayer_44
   WavePlayer_45
   WavePlayer_46
   WavePlayer_47
   WavePlayer_48
   WavePlayer_49
   WavePlayer_50
   WavePlayer_51
  ]
  properties {weights [1,1,1,1,1,1,1,1,1,1,1,1]}
 }
 {
  id=Random_4
  children [WavePlayer_57, WavePlayer_58, WavePlayer_59, WavePlayer_60, WavePlayer_61, WavePlayer_62]
  properties {weights [1,1,1,1,1,1]}
 }
 {id=WavePlayer_40, properties {soundWaveAssetPtr=&/Spot_CrewCut_MilitaryBase_Short_Metal_Creaks_01}}
 {id=WavePlayer_41, properties {soundWaveAssetPtr=&/Spot_CrewCut_MilitaryBase_Short_Metal_Creaks_02}}
 {id=WavePlayer_42, properties {soundWaveAssetPtr=&/Spot_CrewCut_MilitaryBase_Short_Metal_Creaks_03}}
 {id=WavePlayer_43, properties {soundWaveAssetPtr=&/Spot_CrewCut_MilitaryBase_Short_Metal_Creaks_04}}
 {id=WavePlayer_44, properties {soundWaveAssetPtr=&/Spot_CrewCut_MilitaryBase_Short_Metal_Creaks_05}}
 {id=WavePlayer_45, properties {soundWaveAssetPtr=&/Spot_CrewCut_MilitaryBase_Short_Metal_Creaks_06}}
 {id=WavePlayer_46, properties {soundWaveAssetPtr=&/Spot_CrewCut_MilitaryBase_Short_Metal_Creaks_07}}
 {id=WavePlayer_47, properties {soundWaveAssetPtr=&/Spot_CrewCut_MilitaryBase_Short_Metal_Creaks_08}}
 {id=WavePlayer_48, properties {soundWaveAssetPtr=&/Spot_CrewCut_MilitaryBase_Short_Metal_Creaks_09}}
 {id=WavePlayer_49, properties {soundWaveAssetPtr=&/Spot_CrewCut_MilitaryBase_Short_Metal_Creaks_10}}
 {id=WavePlayer_50, properties {soundWaveAssetPtr=&/Spot_CrewCut_MilitaryBase_Short_Metal_Creaks_11}}
 {id=WavePlayer_51, properties {soundWaveAssetPtr=&/Spot_CrewCut_MilitaryBase_Short_Metal_Creaks_12}}
 {id=WavePlayer_57, properties {soundWaveAssetPtr=&Spot_CrewCut_MilitaryBase_Short_Metal_Creaks_01}}
 {id=WavePlayer_58, properties {soundWaveAssetPtr=&Spot_CrewCut_MilitaryBase_Short_Metal_Creaks_02}}
 {id=WavePlayer_59, properties {soundWaveAssetPtr=&Spot_CrewCut_MilitaryBase_Short_Metal_Creaks_03}}
 {id=WavePlayer_60, properties {soundWaveAssetPtr=&Spot_CrewCut_MilitaryBase_Short_Metal_Creaks_04}}
 {id=WavePlayer_61, properties {soundWaveAssetPtr=&Spot_CrewCut_MilitaryBase_Short_Metal_Creaks_05}}
 {id=WavePlayer_62, properties {soundWaveAssetPtr=&Spot_CrewCut_MilitaryBase_Short_Metal_Creaks_06}}
]
properties {
 attenuationSettings=/AmbientAudioBase/Attenuations/Fort_TileAmbient_OneShot_Wider_Att
 bEnableBusSends=false
 bEnableSubmixSends=false
 maxDistance=11000
 soundClassObject=/Game/Sounds/FortSoundClassesAndModes/Fort_Ambient_OneShots_Indoors
 soundSubmixObject=&Fortnite_Reverb_Distant
 volumeMultiplier=0.8
 concurrencyOverrides {maxCount=2}
 concurrencySet [/AmbientAudioBase/Concurrency/AmbientConc_4Max_Spot_Distance_Concurency]
 preEffectBusSends [{}]
}
