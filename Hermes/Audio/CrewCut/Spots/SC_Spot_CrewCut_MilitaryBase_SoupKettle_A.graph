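# Looping ambient spot sound cue for a soup kettle prop at the CrewCut military base, reusing a lava-bubble loop pitched well up.
# A quality-level root feeds one modulator, randomizing pitch and slightly varying volume, into its first two slots and leaves the other two empty.
# A shared medium prop attenuation asset handles falloff.
# Overridden concurrency caps instances, stopping the quietest.
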SoundCue /Hermes/Audio/CrewCut/Spots/SC_Spot_CrewCut_MilitaryBase_SoupKettle_A
root=QualityLevel_1
nodes [
 {id=Modulator_1, children [WavePlayer_1], properties {pitchMax=1.7, pitchMin=1.4, volumeMax=1, volumeMin=0.9}}
 {id=QualityLevel_1, children [Modulator_1,Modulator_1,null,null]}
 {id=WavePlayer_1, properties {bLooping=true, soundWaveAssetPtr=/Game/Sounds/Fort_Starlight/S2/SoundWaves/SpotFX_Lava_Bubbles_Loop}}
]
properties {
 attenuationSettings=/AmbientAudioBase/Attenuations/Fort_Amb_Emitter_Prop_Medium_Att
 bEnableSubmixSends=false
 bOverrideConcurrency=true
 maxDistance=1100
 pitchMultiplier=1.5
 soundClassObject=&Fort_Ambient_Props_Athena
 volumeMultiplier=0.8
 attenuationOverrides {
  bAttenuateWithLPF=true
  bEnableListenerFocus=true
  falloffDistance=950
  lPFFrequencyAtMax=3400
  lPFRadiusMax=1000
  lPFRadiusMin=500
  nonFocusVolumeAttenuation=0.2
  attenuationShapeExtents {x=350}
 }
 concurrencyOverrides {maxCount=4, resolutionRule=StopQuietest, retriggerTime=2, voiceStealReleaseTime=3}
 concurrencySet [/Hermes/Props/SprocketPoppy/Materials/FaceAnim/AmbientConc_3Max_Spot_Concurrency]
}
