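# Looping ambient spot sound cue for the CrewCut military base UV trap, played as a medium prop emitter with low-pass and focus attenuation.
# All quality levels share one branch, which randomly starts at once or after a short delay, then randomizes pitch and boosts volume.
# An endlessly looping envelope slowly modulates pitch over a quick fade-in of one of three random looping waves.
# Concurrency settings limit simultaneous instances.
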SoundCue /Hermes/Audio/CrewCut/Spots/SC_Spot_CrewCut_MilitaryBase_UV_Trap_Loop
root=QualityLevel_1
nodes [
 {id=Delay_0, children [Modulator_1], properties {delayMax=0.3, delayMin=0.3}}
 {
  id=Enveloper_0
  children [Enveloper_4]
  properties {
   bLoop=true
   bLoopIndefinitely=true
   loopEnd=30
   pitchCurve {editorCurveData {keys [{arriveTangent=-0.0033, interpMode=RCIM_Cubic, time=0}, <at 15 1.05 RCIM_Cubic>, {interpMode=RCIM_Cubic, leaveTangent=0.0033, time=30}]}}
   volumeCurve {editorCurveData {keys [<at 0 1>]}}
  }
 }
 {
  id=Enveloper_4
  children [Random_1]
  properties {pitchCurve {editorCurveData {keys [{time=0}]}}, volumeCurve {editorCurveData {keys [<at 0 0>,<at 0.2 1>]}}}
 }
 {id=Modulator_1, children [Enveloper_0], properties {pitchMax=1.1, pitchMin=0.9, volumeMax=1.3, volumeMin=1.3}}
 {id=QualityLevel_1, children [Random_0,Random_0,Random_0,Random_0]}
 {id=Random_0, children [Delay_0,Modulator_1], properties {weights [1,1]}}
 {id=Random_1, children [WavePlayer_1,WavePlayer_2,WavePlayer_3], properties {weights [1,1,1]}}
 {id=WavePlayer_1, properties {bLooping=true, soundWaveAssetPtr=&/Spot_CrewCut_MilitaryBase_UV_Trap_0Loop_01}}
 {id=WavePlayer_2, properties {bLooping=true, soundWaveAssetPtr=&/Spot_CrewCut_MilitaryBase_UV_Trap_0Loop_02}}
 {id=WavePlayer_3, properties {bLooping=true, soundWaveAssetPtr=&/Spot_CrewCut_MilitaryBase_UV_Trap_0Loop_03}}
]
properties {
 attenuationSettings=/AmbientAudioBase/Attenuations/Fort_Amb_Emitter_Prop_Medium_Att
 bEnableSubmixSends=false
 maxDistance=1100
 soundClassObject=&Fort_Ambient_Props_Athena
 volumeMultiplier=0.5
 attenuationOverrides {
  bAttenuateWithLPF=true
  bEnableListenerFocus=true
  falloffDistance=950
  lPFFrequencyAtMax=3400
  lPFRadiusMax=1000
  lPFRadiusMin=500
  nonFocusVolumeAttenuation=0.2
  attenuationShapeExtents {x=350}
 }
 concurrencyOverrides {maxCount=2, retriggerTime=2, voiceStealReleaseTime=3}
 concurrencySet [/Hermes/Props/SprocketPoppy/Materials/FaceAnim/AmbientConc_3Max_Spot_Concurrency]
}
