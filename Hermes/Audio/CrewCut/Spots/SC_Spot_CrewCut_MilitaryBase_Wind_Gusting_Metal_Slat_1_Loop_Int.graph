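# Indoor ambient spot sound for wind gusting through rattling metal at the CrewCut military base.
# A looping vent-rattle bed, randomly one of two waves, fades in and drifts through nested 20-second volume and 30-second pitch envelopes for non-repeating movement.
# A second looping layer plays randomized metal-rattle or low tonal wind one-shots after random delays.
# Cone attenuation adds distance low-pass and listener focus, with concurrency overridden to one instance.
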SoundCue /Hermes/Audio/CrewCut/Spots/SC_Spot_CrewCut_MilitaryBase_Wind_Gusting_Metal_Slat_1_Loop_Int
root=QualityLevel_1
nodes [
 {id=Delay_0, children [Modulator_0], properties {delayMax=8, delayMin=3}}
 {
  id=Enveloper_0
  children [Enveloper_4]
  properties {
   bLoop=true
   bLoopIndefinitely=true
   loopEnd=30
   pitchCurve {editorCurveData {keys [{arriveTangent=-0.0033, interpMode=RCIM_Cubic, time=0}, <at 15 1.05 RCIM_Cubic>, {interpMode=RCIM_Cubic, leaveTangent=0.0033, time=30}]}}
   volumeCurve {editorCurveData {keys [<at 0 1>]}}
  }
 }
 {
  id=Enveloper_3
  children [Enveloper_0]
  properties {
   bLoop=true
   bLoopIndefinitely=true
   loopEnd=20
   pitchCurve {editorCurveData {keys [{time=0}]}}
   volumeCurve {editorCurveData {keys [
      {arriveTangent=-0.01, interpMode=RCIM_Cubic, time=0, value=1}
      {arriveTangent=-0.0666, interpMode=RCIM_Cubic, leaveTangent=-0.0666, time=2.0525, value=0.9761}
      {arriveTangent=-0.0158, interpMode=RCIM_Cubic, leaveTangent=-0.0158, time=5.4253, value=0.6385}
      {arriveTangent=-0.0011, interpMode=RCIM_Cubic, leaveTangent=-0.0011, time=10.0347, value=0.85}
      {arriveTangent=0.0151, interpMode=RCIM_Cubic, leaveTangent=0.0151, time=15.6263, value=0.6268}
      {interpMode=RCIM_Cubic, leaveTangent=0.01, time=20, value=1}
     ]}}
  }
 }
 {
  id=Enveloper_4
  children [Random_0]
  properties {pitchCurve {editorCurveData {keys [{time=0}]}}, volumeCurve {editorCurveData {keys [<at 0 0>,<at 0.2 1>]}}}
 }
 {id=Looping_0, children [Delay_0]}
 {id=Mixer_0, children [Looping_0,Modulator_1], properties {inputVolume [1,1]}}
 {id=Mixer_2, children [Random_6], properties {inputVolume [0.4]}}
 {id=Mixer_3, children [Random_4], properties {inputVolume [0.6]}}
 {id=Mixer_4, children [WavePlayer_2], properties {inputVolume [1.25]}}
 {id=Mixer_5, children [WavePlayer_1], properties {inputVolume [0.8]}}
 {id=Modulator_0, children [Random_5]}
 {id=Modulator_1, children [Enveloper_3], properties {volumeMax=1.3, volumeMin=1.3}}
 {id=QualityLevel_1, children [Mixer_0,Mixer_0,null,null]}
 {id=Random_0, children [Mixer_5,Mixer_4], properties {weights [1,1]}}
 {id=Random_4, children [WavePlayer_11, WavePlayer_12, WavePlayer_13, WavePlayer_14, WavePlayer_15], properties {weights [1,1,1,1,1]}}
 {id=Random_5, children [Mixer_2,Mixer_3], properties {weights [1,1]}}
 {
  id=Random_6
  children [WavePlayer_22, WavePlayer_23, WavePlayer_24, WavePlayer_25, WavePlayer_26, WavePlayer_27]
  properties {weights [1,1,1,1,1,1]}
 }
 {
  id=WavePlayer_1
  properties {bLooping=true, soundWaveAssetPtr=&/Spot_CrewCut_MilitaryBase_Wind_Blustering_Rattling_Building_Vents_01}
 }
 {id=WavePlayer_11, properties {soundWaveAssetPtr=/AmbientAudioContent/Waves/OneShots/Spot_Dojo_Int_Low_Tonal_Wind_03}}
 {id=WavePlayer_12, properties {soundWaveAssetPtr=/AmbientAudioContent/Waves/OneShots/Spot_Dojo_Int_Low_Tonal_Wind_01}}
 {id=WavePlayer_13, properties {soundWaveAssetPtr=/AmbientAudioContent/Waves/OneShots/Spot_Dojo_Int_Low_Tonal_Wind_05}}
 {id=WavePlayer_14, properties {soundWaveAssetPtr=/AmbientAudioContent/Waves/OneShots/Spot_Dojo_Int_Low_Tonal_Wind_04}}
 {id=WavePlayer_15, properties {soundWaveAssetPtr=/AmbientAudioContent/Waves/OneShots/Spot_Dojo_Int_Low_Tonal_Wind_02}}
 {
  id=WavePlayer_2
  properties {bLooping=true, soundWaveAssetPtr=&/Spot_CrewCut_MilitaryBase_Wind_Blustering_Rattling_Building_Vents_02}
 }
 {id=WavePlayer_22, properties {soundWaveAssetPtr=/AmbientAudioContent/Waves/OneShots/Spot_Wind_Metal_Ratte_Gusty_Interior_06}}
 {id=WavePlayer_23, properties {soundWaveAssetPtr=/AmbientAudioContent/Waves/OneShots/Spot_Wind_Metal_Ratte_Gusty_Interior_05}}
 {id=WavePlayer_24, properties {soundWaveAssetPtr=/AmbientAudioContent/Waves/OneShots/Spot_Wind_Metal_Ratte_Gusty_Interior_04}}
 {id=WavePlayer_25, properties {soundWaveAssetPtr=/AmbientAudioContent/Waves/OneShots/Spot_Wind_Metal_Ratte_Gusty_Interior_03}}
 {id=WavePlayer_26, properties {soundWaveAssetPtr=/AmbientAudioContent/Waves/OneShots/Spot_Wind_Metal_Ratte_Gusty_Interior_02}}
 {id=WavePlayer_27, properties {soundWaveAssetPtr=/AmbientAudioContent/Waves/OneShots/Spot_Wind_Metal_Ratte_Gusty_Interior_01}}
]
properties {
 bOverrideAttenuation=true
 bOverrideConcurrency=true
 maxDistance=1100
 soundClassObject=&Fort_Ambient_Indoors_Athena
 volumeMultiplier=1
 attenuationOverrides {
  attenuationShape=Cone
  bAttenuateWithLPF=true
  bEnableFocusInterpolation=true
  bEnableListenerFocus=true
  falloffDistance=750
  focusAttackInterpSpeed=0.05
  focusReleaseInterpSpeed=0.01
  lPFFrequencyAtMax=3400
  lPFRadiusMax=1000
  lPFRadiusMin=500
  nonFocusVolumeAttenuation=0.4
  attenuationShapeExtents=<350 60 90>
 }
 concurrencyOverrides {maxCount=1, retriggerTime=2, voiceStealReleaseTime=3}
 concurrencySet [/Hermes/Props/SprocketPoppy/Materials/FaceAnim/AmbientConc_3Max_Spot_Concurrency]
}
