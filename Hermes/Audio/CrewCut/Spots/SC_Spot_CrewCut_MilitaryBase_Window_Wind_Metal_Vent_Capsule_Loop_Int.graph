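# Indoor ambient spot SoundCue for wind rattling metal vents at the CrewCut military base, with occluded, focus-aware capsule attenuation.
# A looping bed randomly picks one of two vent waves, fades it in, and varies it through nested looping volume and pitch envelopes of differing lengths.
# A second looping branch mixes in random gusty metal rattle one-shots after random delays.
# The last two quality-level slots are empty, leaving them silent.
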SoundCue /Hermes/Audio/CrewCut/Spots/SC_Spot_CrewCut_MilitaryBase_Window_Wind_Metal_Vent_Capsule_Loop_Int
root=QualityLevel_1
nodes [
 {id=Delay_0, children [Modulator_0], properties {delayMax=8, delayMin=3}}
 {
  id=Enveloper_6
  children [Enveloper_8]
  properties {
   bLoop=true
   bLoopIndefinitely=true
   loopEnd=30
   pitchCurve {editorCurveData {keys [{arriveTangent=-0.0033, interpMode=RCIM_Cubic, time=0}, <at 15 1.05 RCIM_Cubic>, {interpMode=RCIM_Cubic, leaveTangent=0.0033, time=30}]}}
   volumeCurve {editorCurveData {keys [<at 0 1>]}}
  }
 }
 {
  id=Enveloper_7
  children [Enveloper_6]
  properties {
   bLoop=true
   bLoopIndefinitely=true
   loopEnd=20
   pitchCurve {editorCurveData {keys [{time=0}]}}
   volumeCurve {editorCurveData {keys [
      {arriveTangent=-0.01, interpMode=RCIM_Cubic, time=0, value=1}
      {arriveTangent=-0.0769, interpMode=RCIM_Cubic, leaveTangent=-0.0769, time=2.0525, value=0.9761}
      {arriveTangent=-0.0158, interpMode=RCIM_Cubic, leaveTangent=-0.0158, time=6.3285, value=0.5135}
      {arriveTangent=-0.0011, interpMode=RCIM_Cubic, leaveTangent=-0.0011, time=10.0347, value=0.85}
      {arriveTangent=0.0151, interpMode=RCIM_Cubic, leaveTangent=0.0151, time=16.5294, value=0.5019}
      {interpMode=RCIM_Cubic, leaveTangent=0.01, time=20, value=1}
     ]}}
  }
 }
 {
  id=Enveloper_8
  children [Random_0]
  properties {pitchCurve {editorCurveData {keys [{time=0}]}}, volumeCurve {editorCurveData {keys [<at 0 0>,<at 0.2 1>]}}}
 }
 {id=Looping_0, children [Delay_0]}
 {id=Mixer_0, children [Looping_0,Modulator_3], properties {inputVolume [1,1]}}
 {id=Mixer_2, children [Random_6], properties {inputVolume [0.3]}}
 {id=Mixer_4, children [WavePlayer_2], properties {inputVolume [1.25]}}
 {id=Mixer_5, children [WavePlayer_1], properties {inputVolume [0.8]}}
 {id=Modulator_0, children [Random_5]}
 {id=Modulator_3, children [Enveloper_7], properties {volumeMax=1.3, volumeMin=1.3}}
 {id=QualityLevel_1, children [Mixer_0,Mixer_0,null,null]}
 {id=Random_0, children [Mixer_5,Mixer_4], properties {weights [1,1]}}
 {id=Random_5, children [Mixer_2], properties {weights [1]}}
 {
  id=Random_6
  children [WavePlayer_22, WavePlayer_23, WavePlayer_24, WavePlayer_25, WavePlayer_26, WavePlayer_27]
  properties {weights [1,1,1,1,1,1]}
 }
 {
  id=WavePlayer_1
  properties {bLooping=true, soundWaveAssetPtr=&/Spot_CrewCut_MilitaryBase_Wind_Blustering_Rattling_Building_Vents_01}
 }
 {
  id=WavePlayer_2
  properties {bLooping=true, soundWaveAssetPtr=&/Spot_CrewCut_MilitaryBase_Wind_Blustering_Rattling_Building_Vents_02}
 }
 {id=WavePlayer_22, properties {soundWaveAssetPtr=/AmbientAudioContent/Waves/OneShots/Spot_Wind_Metal_Ratte_Gusty_Interior_06}}
 {id=WavePlayer_23, properties {soundWaveAssetPtr=/AmbientAudioContent/Waves/OneShots/Spot_Wind_Metal_Ratte_Gusty_Interior_05}}
 {id=WavePlayer_24, properties {soundWaveAssetPtr=/AmbientAudioContent/Waves/OneShots/Spot_Wind_Metal_Ratte_Gusty_Interior_04}}
 {id=WavePlayer_25, properties {soundWaveAssetPtr=/AmbientAudioContent/Waves/OneShots/Spot_Wind_Metal_Ratte_Gusty_Interior_03}}
 {id=WavePlayer_26, properties {soundWaveAssetPtr=/AmbientAudioContent/Waves/OneShots/Spot_Wind_Metal_Ratte_Gusty_Interior_02}}
 {id=WavePlayer_27, properties {soundWaveAssetPtr=/AmbientAudioContent/Waves/OneShots/Spot_Wind_Metal_Ratte_Gusty_Interior_01}}
]
properties {
 attenuationSettings=/AmbientAudioBase/Attenuations/Fort_Amb_Emitter_Prop_Medium_Att
 bEnableSubmixSends=false
 bOverrideAttenuation=true
 bOverrideConcurrency=true
 maxDistance=950
 pitchMultiplier=1.5
 soundClassObject=&Fort_Ambient_Indoors_Athena
 volumeMultiplier=0.8
 attenuationOverrides {
  attenuationShape=Capsule
  bAttenuateWithLPF=true
  bEnableFocusInterpolation=true
  bEnableListenerFocus=true
  bEnableOcclusion=true
  distanceAlgorithm=Logarithmic
  falloffDistance=650
  focusAttackInterpSpeed=0.05
  focusDistanceScale=0.9
  focusReleaseInterpSpeed=0.01
  lPFFrequencyAtMax=3400
  lPFRadiusMax=1000
  lPFRadiusMin=500
  nonFocusVolumeAttenuation=0.4
  occlusionInterpolationTime=3
  occlusionLowPassFilterFrequency=850
  occlusionTraceChannel=ECC_GameTraceChannel11
  attenuationShapeExtents=<300 98.0462>
 }
 concurrencyOverrides {maxCount=2, retriggerTime=2, voiceStealReleaseTime=3}
 concurrencySet [/Hermes/Props/SprocketPoppy/Materials/FaceAnim/AmbientConc_3Max_Spot_Concurrency]
}
